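# Niagara system for the CrumbViolin emote that plays short-lived yellow-green electric mesh arcs over the character's body.
# A single BodyArcs emitter spawns curve-driven arcs whose random spawn group sets visibility tag, bone set, cylinder radius and orientation.
# Arcs sit on cylinder surfaces around filtered skeletal bones, using ten mesh renderers, animated material parameters and flickering brightness.
# An age-driven kill sphere expands, then collapses early in playback.
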NiagaraSystem /BRCosmetics/Animation/Game/MainPlayer/Emotes/CrumbViolin/FX/NS_Emote_CrumbViolin_Electricity
dependencies {
 usedDataInterfaces [&NiagaraDataInterfaceCurve, &NiagaraDataInterfaceArrayInt32, &NiagaraDataInterfaceSkeletalMesh]
 usedDynamicInputs []
 usedModules [
  &self:SystemSpawnScript.SystemScriptSource.SystemScriptGraph.NiagaraNodeAssignment_0.SetVariables_#1
  &self:SystemSpawnScript.SystemScriptSource.SystemScriptGraph.NiagaraNodeAssignment_1.SetVariables_#2
  &SystemState
 ]
}
emitters {BodyArcs {
  emitterUpdateScript [
   &EmitterState()
   &SpawnRate(
    Spawn Probability=0.7
    SpawnGroup:NiagaraInt32=&SelectIntFromArray(
     Array Sampling Mode=::NewEnumerator1
     Direct Array Index=?
     Evaluation Type=::NewEnumerator1
     Int Selection Array=DataInterface() {propertyValues {IntData [0,1,2,3,4,0,1,2]}}
     Random Seed=0
    )
    SpawnRate=&Multiply_Float(
     A=43
     B=&FloatFromCurve(
      FloatCurve=DataInterface() {propertyValues {Curve {keys [
          {arriveTangent=1.8415, interpMode=RCIM_Cubic, leaveTangent=1.8415, tangentMode=RCTM_User, time=0, value=0.45}
          {arriveTangent=-7.2882, interpMode=RCIM_Cubic, leaveTangent=0.1479, tangentMode=RCTM_Break, time=0.3, value=0.05}
          {arriveTangent=9, interpMode=RCIM_Cubic, leaveTangent=9, time=0.3454, value=0.1}
          {arriveTangent=8.802, interpMode=RCIM_Cubic, leaveTangent=0.4887, tangentMode=RCTM_Break, time=0.4, value=0.95}
          {arriveTangent=-0.0405, interpMode=RCIM_Cubic, leaveTangent=-0.0405, tangentMode=RCTM_User, time=0.7009, value=1}
          {arriveTangent=-8.0089, interpMode=RCIM_Cubic, leaveTangent=-8.0089, tangentMode=RCTM_User, time=1, value=0}
         ]}}}
      CurveIndex=$Emitter.NormalizedLoopAge
     )
    )
    Use Spawn Probability=true
   )
   &SpawnBurst_Instantaneous(
    Spawn Group:NiagaraInt32=&SelectIntFromArray(
     Array Sampling Mode=::NewEnumerator1
     Direct Array Index=?
     Evaluation Type=::NewEnumerator1
     Int Selection Array=DataInterface() {propertyValues {IntData [0,1,2,3,4]}}
     Random Seed=0
    )
    Use Spawn Probability=false
    UseLoopCountLimit=false
   )
  ]
  particleSpawnScript [
   /SetVariables(Particles.VisibilityTag=&SpawnGroup()) {isSetParameters=true}
   &InitializeParticle(
    Color=<rgba 12.9459 15 0 1>
    Lifetime Max=0.35
    Lifetime Min=0.175
    Lifetime Mode=::NewEnumerator1
    Mesh Scale Min=<2 2 1>
    Mesh Scale Mode=::NewEnumerator2
    UsePositionOffset=false
    Write Material Random=true
   )
   &InitialMeshOrientation(
    Mesh Orientation Mode=::NewEnumerator2
    Rotation=&MaskVectorBySpawnGroup(
     Masked Vector=<-0.03 0.255 0>
     Passthrough Vector=&MaskVectorBySpawnGroup(
      Masked Vector=<-0.03 -0.255 0>
      Passthrough Vector=&MaskVectorBySpawnGroup(
       Masked Vector=<0 0.25 0>
       Passthrough Vector=&MaskVectorBySpawnGroup(Masked Vector=<-0.05 -0.25 0>, Passthrough Vector=<0.05 0.25 0>, Spawn Group Mask=3)
       Spawn Group Mask=2
      )
      Spawn Group Mask=1
     )
    )
    Use Rotation Vector=true
   )
  ]
  particleUpdateScript [
   &LinearForce(Force=<3 0 0>)
   &SolveForcesAndVelocity(Clamp Velocity=false, Limit Acceleration=false)
   &SkeletalMeshLocation(
    Bone / Socket Index=&MaskIntBySpawnGroup(
     Masked Int=&SelectIntFromArray(
      Array Sampling Mode=::NewEnumerator1
      Direct Array Index=?
      Int Selection Array=DataInterface() {propertyValues {IntData [0,1]}}
      Random Seed=0
     )
     Passthrough Int=&MaskIntBySpawnGroup(
      Masked Int=&SelectIntFromArray(
       Array Sampling Mode=::NewEnumerator1
       Direct Array Index=?
       Int Selection Array=DataInterface() {propertyValues {IntData [2,3]}}
       Random Seed=0
      )
      Passthrough Int=&MaskIntBySpawnGroup(
       Masked Int=&SelectIntFromArray(
        Array Sampling Mode=::NewEnumerator1
        Direct Array Index=?
        Int Selection Array=DataInterface() {propertyValues {IntData [6,7,8]}}
        Random Seed=0
       )
       Passthrough Int=&MaskIntBySpawnGroup(
        Masked Int=&SelectIntFromArray(
         Array Sampling Mode=::NewEnumerator1
         Direct Array Index=?
         Int Selection Array=DataInterface() {propertyValues {IntData [9,10,12]}}
         Random Seed=0
        )
        Passthrough Int=&SelectIntFromArray(
         Array Sampling Mode=::NewEnumerator1
         Direct Array Index=?
         Int Selection Array=DataInterface() {propertyValues {IntData [4,5,11]}}
         Random Seed=0
        )
        Spawn Group Mask=3
       )
       Spawn Group Mask=2
      )
      Spawn Group Mask=1
     )
     Spawn Group Mask=0
    )
    Bone Sampling Mode=::NewEnumerator4
    MaskSpawnGroup=false
    Skeletal Mesh=DataInterface() {propertyValues {
      DefaultMesh=null
      ExcludeBoneName=root
      OptionalComponentClass=null
      PreviewMesh=/BRCosmetics/Characters/Player/Male/Large/Bodies/M_LRG_CrumbViolin/Meshes/M_LRG_CrumbViolin
      SkinningMode=SkinOnTheFly
      SoftSourceActor=null
      SourceMode=Default
      UvSetIndex=0
      WholeMeshLOD=-1
      bExcludeBone=false
      bReadDeformedGeometry=true
      bRequireCurrentFrameData=true
      ComponentTags []
      FilteredBones [lowerarm_l,upperarm_l,lowerarm_r,upperarm_r,thigh_l,calf_l,pelvis,spine_01,spine_05,thigh_r,calf_r,foot_l,foot_r]
      FilteredSockets []
      MeshUserParameter {parameter {name=None, typeDefHandle {registeredTypeIndex=1}, varData []}}
      SamplingRegions []
     }}
   )
   &ShapeLocation(
    Cylinder Height=18
    Cylinder Radius=&MaskFloatBySpawnGroup(
     Masked Float=5
     Passthrough Float=&MaskFloatBySpawnGroup(
      Masked Float=5
      Passthrough Float=&MaskFloatBySpawnGroup(
       Masked Float=20
       Passthrough Float=&MaskFloatBySpawnGroup(Masked Float=8, Passthrough Float=8, Spawn Group Mask=3)
       Spawn Group Mask=2
      )
      Spawn Group Mask=1
     )
    )
    MaskSpawnGroup=false
    Override Seed=false
    Shape Origin=$StackContext.SkeletalMeshLocation.SampledPosition
    Shape Primitive=::NewEnumerator1
    Surface Only=true
    Surface Only Band Thickness=0.5
    UseOwnerScale=true
   )
   &ParticleState()
   &ScaleMeshSize(Scale Factor=&Multiply_VectorByFloat(
     Vector=&MakeCustomVectorFromBool(
      Bool=&SetBoolByFloatComparison(
       A=&VectorLength(Vector=Linked() {linkedVariable {name=Particles.Position, type=NiagaraPosition}})
       B=&VectorLength(Vector=<0 0 85>)
       Comparison Type=::NewEnumerator2
      )
      True Vector=<1 1 1>
      False Vector=&Add_Vector(A=$Particles.Scale, B=&RandomRangeVector(Minimum=<0.6 0.6 1>, Maximum=<0.6 0.6 1.6>))
     )
     Float=0.1
    )) {module=ScaleMeshSize001}
   &ScaleMeshSize(Scale Factor=&VectorFromFloat(Value=&FloatFromCurve(
      FloatCurve=DataInterface() {propertyValues {Curve {keys [
          {arriveTangent=0.1331, interpMode=RCIM_Cubic, leaveTangent=0.1331, tangentMode=RCTM_User, time=0.3, value=0.6}
          {interpMode=RCIM_Cubic, tangentMode=RCTM_User, time=1, value=1}
         ]}}}
      CurveIndex=$System.Age
     )))
   &DynamicMaterialParameters(
    Index 0 Param 1=&MakeFloatFromVector(Vector=$Particles.Scale, Channel=::NewEnumerator0)
    Index 0 Param 2=&RandomRangeFloat(Minimum=0, Maximum=10)
    Index 0 Param 3=&FloatFromCurve(
     FloatCurve=DataInterface() {propertyValues {Curve {keys [{interpMode=RCIM_Cubic, leaveTangent=8.1085, tangentMode=RCTM_Break, time=0.8568, value=1}, {tangentMode=RCTM_User, time=1, value=0}]}}}
     Scale Curve=&RandomRangeFloat(Minimum=0.015, Maximum=0.025, Evaluation Type=::NewEnumerator1)
    )
    Index 0 Param 4=0.8
    Index 1 Param 1=&FloatFromCurve(FloatCurve=DataInterface() {propertyValues {Curve {keys [
         {interpMode=RCIM_Cubic, leaveTangent=1.1034, tangentMode=RCTM_Break, time=0, value=0}
         {arriveTangent=2.2399, tangentMode=RCTM_User, time=0.5, value=1}
        ]}}})
    Index 1 Param 2=&FloatFromCurve(FloatCurve=DataInterface() {propertyValues {Curve {keys [<at 0.5 0>, {arriveTangent=3.2053, tangentMode=RCTM_User, time=0.9, value=1}]}}})
    Index 1 Param 3=&FloatFromCurve(FloatCurve=DataInterface() {propertyValues {Curve {keys [<at 0 0 RCIM_Cubic>, {arriveTangent=3.4223, interpMode=RCIM_Cubic, leaveTangent=3.4223, tangentMode=RCTM_User, time=1, value=1}]}}})
    Index 1 Param 4=&FloatFromCurve(
     FloatCurve=DataInterface() {propertyValues {Curve {keys [
         {interpMode=RCIM_Cubic, leaveTangent=0.1169, tangentMode=RCTM_Break, time=0, value=0}
         {arriveTangent=2.5708, interpMode=RCIM_Cubic, leaveTangent=2.5708, tangentMode=RCTM_User, time=1, value=1}
        ]}}}
     Scale Curve=&RandomRangeFloat(Minimum=1, Maximum=40)
    )
    Index 2 Param 1=&FloatFromCurve(FloatCurve=DataInterface() {propertyValues {Curve {keys [
         {interpMode=RCIM_Cubic, leaveTangent=-0.0025, tangentMode=RCTM_Break, time=0, value=0.5}
         {arriveTangent=-0.0316, interpMode=RCIM_Cubic, tangentMode=RCTM_Break, time=1, value=-0.2}
        ]}}})
    Index 2 Param 2=&FloatFromCurve(
     FloatCurve=DataInterface() {propertyValues {Curve {keys [
         {interpMode=RCIM_Cubic, tangentMode=RCTM_User, time=0, value=0}
         {arriveTangent=2.0522, interpMode=RCIM_Cubic, leaveTangent=2.0522, tangentMode=RCTM_User, time=1, value=1}
        ]}}}
     Scale Curve=2
    )
    Module.Index0Param0Write=true
    Module.Index0Param1Write=true
    Module.Index0Param2Write=true
    Module.Index0Param3Write=true
    Module.Index1Param0Write=true
    Module.Index1Param1Write=true
    Module.Index1Param2Write=true
    Module.Index1Param3Write=true
    Module.Index2Param0Write=true
    Module.Index2Param1Write=true
    Module.Index2Param2Write=false
    Module.Index2Param3Write=false
    Module.Index3Param0Write=false
    Module.Index3Param1Write=false
    Module.Index3Param2Write=false
    Module.Index3Param3Write=true
    Write Parameter Index 1=true
    Write Parameter Index 2=true
    Write Parameter Index 3=true
   )
   &ScaleColor(
    Scale Alpha=&FloatFromCurve(FloatCurve=DataInterface() {propertyValues {Curve {keys [
         {tangentMode=RCTM_User, time=0, value=1}
         {arriveTangent=-0.0405, interpMode=RCIM_Cubic, leaveTangent=-0.0405, tangentMode=RCTM_User, time=0.7009, value=1}
         {arriveTangent=-8.0089, interpMode=RCIM_Cubic, leaveTangent=-8.0089, tangentMode=RCTM_User, time=1, value=0}
        ]}}})
    Scale RGB=&Lerp_Vector(
     A=&VectorFromFloat(Value=&FloatFromCurve(
       FloatCurve=DataInterface() {propertyValues {Curve {keys [<at 0 1>,<at 0.1 1>,<at 0.2 2>,<at 0.3 1>,<at 0.6 1>,<at 0.7 2>,<at 0.8 1>,<at 1 1>]}}}
       Scale Curve=50
      ))
     B=&VectorFromFloat(Value=&RandomRangeFloat(Minimum=0.3, Maximum=5, Evaluation Type=::NewEnumerator1))
     Alpha=&FloatFromCurve(FloatCurve=DataInterface() {propertyValues {Curve {keys [
          {interpMode=RCIM_Cubic, leaveTangent=2.5294, tangentMode=RCTM_Break, time=0, value=0}
          {arriveTangent=-0.2166, interpMode=RCIM_Cubic, tangentMode=RCTM_Break, time=0.4, value=1}
         ]}}})
    )
    ScaleA=true
    ScaleRGB=true
   )
   &CameraOffset(Camera Offset Amount=&MaskFloat(
     Mask Parameter=&SpawnGroup()
     Mask=2
     Masked Float=&RandomRangeFloat(Minimum=-1, Maximum=15)
     Passthrough Float=&RandomRangeFloat(Minimum=-1, Maximum=15)
    ))
   /SetVariables(Particles.Position=&Lerp_Vector(
     A=Linked() {linkedVariable {name=Particles.Position, type=NiagaraPosition}}
     B=&AddVectorToPosition(Position=$Particles.Position, Vector=&RandomRangeVector(Minimum=<0 0 1>, Maximum=<0 0 5>))
     Alpha=&FloatFromCurve(FloatCurve=DataInterface() {propertyValues {Curve {keys [
          {arriveTangent=1.1702, interpMode=RCIM_Cubic, leaveTangent=1.1702, tangentMode=RCTM_User, time=0, value=0}
          {interpMode=RCIM_Cubic, tangentMode=RCTM_User, time=1, value=1}
         ]}}})
    )) {isSetParameters=true}
   &KillParticlesInVolume(Sphere Radius:NiagaraFloat=&Multiply_Float(
     A=150
     B=&FloatFromCurve(
      FloatCurve=DataInterface() {propertyValues {Curve {keys [
          {arriveTangent=14.7597, interpMode=RCIM_Cubic, leaveTangent=14.7597, tangentMode=RCTM_User, time=0, value=0}
          {arriveTangent=0.0951, interpMode=RCIM_Cubic, leaveTangent=0.0951, tangentMode=RCTM_User, time=0.1408, value=0.9847}
          {arriveTangent=-4.7423, interpMode=RCIM_Cubic, leaveTangent=-4.7423, tangentMode=RCTM_User, time=0.4, value=0.675}
          {arriveTangent=-9.3869, leaveTangent=-1.125, tangentMode=RCTM_User, time=0.5, value=0}
          {arriveTangent=-0.0405, interpMode=RCIM_Cubic, leaveTangent=-0.0405, tangentMode=RCTM_User, time=1, value=0}
         ]}}}
      CurveIndex=$Engine.System.Age
     )
    ))
  ]
  rendererClasses [&NiagaraMeshRendererProperties]
  renderers [
   {rendererClass=&NiagaraMeshRendererProperties, rendererIndex=0}
   {rendererClass=&NiagaraMeshRendererProperties, rendererIndex=1}
   {rendererClass=&NiagaraMeshRendererProperties, rendererIndex=2}
   {rendererClass=&NiagaraMeshRendererProperties, rendererIndex=3}
   {rendererClass=&NiagaraMeshRendererProperties, rendererIndex=4}
   {rendererClass=&NiagaraMeshRendererProperties, rendererIndex=5}
   {rendererClass=&NiagaraMeshRendererProperties, rendererIndex=6}
   {rendererClass=&NiagaraMeshRendererProperties, rendererIndex=7}
   {rendererClass=&NiagaraMeshRendererProperties, rendererIndex=8}
   {rendererClass=&NiagaraMeshRendererProperties, rendererIndex=9}
  ]
 }}
summary {
 systemName=NS_Emote_CrumbViolin_Electricity
 emitters [{emitterName=BodyArcs, rendererClasses [&NiagaraMeshRendererProperties]}]
 userVariables []
}
