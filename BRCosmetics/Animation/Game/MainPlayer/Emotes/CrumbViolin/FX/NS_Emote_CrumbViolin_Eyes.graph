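# Niagara system for the CrumbViolin emote that adds short-lived glowing eyes to the player's skeletal mesh, with a matching light.
# A GPU sprite emitter bursts two spawn groups, follows sampled bones or sockets, and fades with flickering brightness, scaled by time of day.
# Sprite size lerps by camera-to-particle dot product against the Y axis, reshaping the glow with viewing angle.
# Emitter LR drives a light renderer with a time-of-day-scaled pulse.
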NiagaraSystem /BRCosmetics/Animation/Game/MainPlayer/Emotes/CrumbViolin/FX/NS_Emote_CrumbViolin_Eyes
dependencies {
 usedDataInterfaces [&NiagaraDataInterfaceCamera, &NiagaraDataInterfaceCurve]
 usedDynamicInputs [
  &MakeFloatFromVector2d
  &ClampFloat
  &Subtract_Float
  &MakeFloatFromVector
  &RotateVectorByQuaternion
 ]
 usedModules [
  &SystemState
  &SampleSkeletalMesh
  &self:SystemSpawnScript.SystemScriptSource.SystemScriptGraph.NiagaraNodeAssignment_2.SetVariables_#1
  &self:SystemSpawnScript.SystemScriptSource.SystemScriptGraph.NiagaraNodeAssignment_0.SetVariables_#2
 ]
}
emitters {
 Glow_Eyes003 {
  simTarget=GPUComputeSim
  emitterUpdateScript [
   &EmitterState()
   &SpawnBurst_Instantaneous(Use Spawn Probability=false, UseLoopCountLimit=false)
   &SpawnBurst_Instantaneous(Spawn Count=2, Spawn Group=1, Use Spawn Probability=false, UseLoopCountLimit=false) {module=SpawnBurst_Instantaneous001}
  ]
  particleSpawnScript [
   &InitializeParticle(
    Color=&MaskLinearColorBySpawnGroup(Masked Linear Color=<rgba 6 0 0.3161 1>, Passthrough Linear Color=<rgba 15 0.2121 0 1>)
    Lifetime=1.8
    Sprite Size=&MaskVector2dBySpawnGroup(
     Masked Vector2D=&Multiply_Vector2D_ByFloat(Float=1.5, Vector2D=<55 25>)
     Passthrough Vector2D=&Multiply_Vector2D_ByFloat(Float=1, Vector2D=<35 15>)
     Spawn Group Mask=0
    )
    Sprite Size Mode=::NewEnumerator4
    UsePositionOffset=false
    Write Material Random=true
   )
   &CameraOffset(Camera Offset Amount=5)
  ]
  particleUpdateScript [
   &SolveForcesAndVelocity(Clamp Velocity=false, Limit Acceleration=false)
   /SetVariables(Particles.CamToParticleVector:Vector3f=&NormalizeVector(
     Vector To Normalize=&SubtractVectorFromPosition(
      Position=&ReturnCameraPosition(Coordinate Space:ENiagaraCoordinateSpace=::Local, Camera Query:NiagaraDataInterfaceCamera=DataInterface())
      Vector=Linked() {linkedVariable {name=Particles.Position, type=NiagaraPosition}}
     )
     Safe Normalize=true
     Safe Fallback Threshold=0
     Safe Fallback Vector=<0 0 0.001>
    )) {isSetParameters=true}
   /SetVariables(Particles.DotVector:NiagaraFloat=&Abs_Float(Float:NiagaraFloat=&DotProduct(Vector A=$Particles.CamToParticleVector, Vector B=<0 1 0>))) {isSetParameters=true}
   &ParticleState()
   &SkeletalMeshLocation(Bone / Socket Sampling Mode=::NewEnumerator5, MaskSpawnGroup=false, Mesh Sampling Type=::NewEnumerator6, Skeletal Mesh=$User.SkeletalMesh)
   /SetVariables(Particles.Position=&AddVectorToPosition(Position=$Particles.Position, Vector=<0 -2 0.2>)) {isSetParameters=true}
   /SetVariables(Particles.SpriteSize=&Lerp_Vector2(
     A=$Particles.Initial.SpriteSize
     B=&MaskVector2dBySpawnGroup(Masked Vector2D=<25 25>, Passthrough Vector2D=<10 10>, Spawn Group Mask=0)
     Alpha=&FloatFromCurve(
      FloatCurve=DataInterface() {propertyValues {Curve {keys [{interpMode=RCIM_Cubic, tangentMode=RCTM_User, time=0, value=1}, {interpMode=RCIM_Cubic, tangentMode=RCTM_User, time=1, value=0}]}}}
      CurveIndex=$Particles.DotVector
     )
    )) {isSetParameters=true}
   &ScaleColor(
    Scale Alpha=&FloatFromCurve(
     FloatCurve=DataInterface() {propertyValues {Curve {keys [
         {
          arriveTangent=9.0725
          interpMode=RCIM_Cubic
          leaveTangent=9.0725
          leaveTangentWeight=0.1521
          tangentMode=RCTM_User
          tangentWeightMode=RCTWM_WeightedBoth
          time=0
          value=0
         }
         {
          arriveTangent=-3.4704
          arriveTangentWeight=0.0591
          interpMode=RCIM_Cubic
          leaveTangent=-3.4704
          leaveTangentWeight=1.1238
          tangentMode=RCTM_User
          time=0.025
          value=1
         }
         {arriveTangent=0.3426, interpMode=RCIM_Cubic, leaveTangent=2.3193, tangentMode=RCTM_Break, time=0.5, value=0.5}
         {
          arriveTangent=-3.3787
          arriveTangentWeight=0.5432
          interpMode=RCIM_Cubic
          leaveTangent=-3.3787
          leaveTangentWeight=0.5432
          tangentMode=RCTM_User
          tangentWeightMode=RCTWM_WeightedBoth
          time=1
          value=0
         }
        ]}}}
     Scale Curve=&Multiply_Float(A=&FNTimeOfDayScalar(Night Scale=0.7, Morning Scale=0.8, Evening Scale=0.8), B=0.7)
    )
    Scale RGB=&Lerp_Vector(
     A=&VectorFromFloat(Value=&FloatFromCurve(
       FloatCurve=DataInterface() {propertyValues {Curve {keys [<at 0 0>,<at 0.1 4>,<at 0.5 2.75>,<at 0.6 1.5>,<at 0.7 3.25>,<at 0.8 1>,<at 0.9 1.15>,<at 1 1>]}}}
       Scale Curve=0.0001
      ))
     B=&VectorFromFloat(Value=&RandomRangeFloat(Minimum=0.3, Maximum=1, Evaluation Type=::NewEnumerator1))
     Alpha=&FloatFromCurve(FloatCurve=DataInterface() {propertyValues {Curve {keys [<at 0 0 RCIM_Cubic>,<at 1 1 RCIM_Cubic>]}}})
    )
    ScaleA=true
    ScaleRGB=true
   )
   &FN_Light_Attributes(
    Color=<rgba 6 0 0.3161 1>
    Create Color Attribute=true
    Create Exponent Attribute=true
    Create Position Attribute=true
    Create Radius Attribute=true
    Exponent=3
    Position=$Particles.SkeletalMeshLocation.SampledPosition
    Radius=17
    Scale Color=&FloatFromCurve(
     FloatCurve=DataInterface() {propertyValues {Curve {keys [
         {arriveTangent=1.8576, interpMode=RCIM_Cubic, leaveTangent=1.8576, tangentMode=RCTM_User, time=0, value=0}
         {interpMode=RCIM_Cubic, leaveTangent=-3.7152, tangentMode=RCTM_Break, time=0.5, value=1}
         {tangentMode=RCTM_User, time=1, value=0}
        ]}}}
     Scale Curve=15
    )
   )
   &DynamicMaterialParameters(Write Parameter Index 0=::NewEnumerator0)
  ]
  rendererClasses [&NiagaraSpriteRendererProperties]
  renderers [{rendererClass=&NiagaraSpriteRendererProperties}]
 }
 LR {
  emitterUpdateScript [
   &EmitterState()
   &SpawnBurst_Instantaneous(Use Spawn Probability=false, UseLoopCountLimit=false)
  ]
  particleSpawnScript [&InitializeParticle(Lifetime=1.8, UsePositionOffset=false, Write Material Random=true)]
  particleUpdateScript [
   &ParticleState()
   &SkeletalMeshLocation(Bone / Socket Sampling Mode=::NewEnumerator5, MaskSpawnGroup=false, Mesh Sampling Type=::NewEnumerator6, Skeletal Mesh=$User.SkeletalMesh)
   &FN_Light_Attributes(
    Color=<rgba 6 0 0.3161 1>
    Create Color Attribute=true
    Create Exponent Attribute=true
    Create Position Attribute=true
    Create Radius Attribute=true
    Exponent=3
    Position=&AddVectorToPosition(Position=$Particles.SkeletalMeshLocation.SampledPosition, Vector=<0 2 0>)
    Radius=80
    Scale Color=&FloatFromCurve(
     FloatCurve=DataInterface() {propertyValues {Curve {keys [
         {arriveTangent=1.8576, interpMode=RCIM_Cubic, leaveTangent=1.8576, tangentMode=RCTM_User, time=0, value=0}
         {interpMode=RCIM_Cubic, leaveTangent=-3.7152, tangentMode=RCTM_Break, time=0.5, value=1}
         {tangentMode=RCTM_User, time=1, value=0}
        ]}}}
     Scale Curve=&Multiply_Float(A=3, B=$System.ToD)
    )
   )
   &SolveForcesAndVelocity(Clamp Velocity=false, Limit Acceleration=false)
  ]
  rendererClasses [&NiagaraLightRendererProperties]
  renderers [{rendererClass=&NiagaraLightRendererProperties, rendererIndex=0}]
 }
}
summary {
 systemName=NS_Emote_CrumbViolin_Eyes
 emitters [
  {emitterName=Glow_Eyes003, simTarget=GPUComputeSim, rendererClasses [&NiagaraSpriteRendererProperties]}
  {emitterName=LR, rendererClasses [&NiagaraLightRendererProperties]}
 ]
 userVariables [
  User.bIsFrontEnd {type=NiagaraBool}
  User.bIsFrontEndPreview {type=NiagaraBool}
  User.MaskOff {type=NiagaraFloat, defaultValue=0, description=""}
  User.SkeletalMesh {
   type=NiagaraDataInterfaceSkeletalMesh
   defaultValue {
    struct=NiagaraExt_VariableValue_DataInterface
    value {dataInterface=&self:NiagaraDataInterfaceSkeletalMesh_0}
   }
  }
  User.Style {type=NiagaraFloat, defaultValue=0, description=""}
 ]
}
