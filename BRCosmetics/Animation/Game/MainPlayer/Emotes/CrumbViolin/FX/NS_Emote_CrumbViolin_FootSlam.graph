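# Niagara system for the foot slam in the CrumbViolin emote, bursting dust, small debris and short streaks at the impact.
# GroundDust lerps between two tints via User.DesertBoulder, SmallBits simulates gravity-driven GPU sprites, and Streaks adds brief distance-culled sprites.
# MainBurst computes ribbon and sprite positions along a length driven by User.Reveal, rotated by System.Foot_Rotator, with spawn groups separating particle behaviors.
# Time-of-day functions adjust dust opacity and burst color.
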NiagaraSystem /BRCosmetics/Animation/Game/MainPlayer/Emotes/CrumbViolin/FX/NS_Emote_CrumbViolin_FootSlam
dependencies {
 usedDataInterfaces [
  &NiagaraDataInterfaceCurve
  &NiagaraDataInterfaceVector2DCurve
  &NiagaraDataInterfaceVectorCurve
  &NiagaraDataInterfaceColorCurve
  &NiagaraDataInterfaceSpriteRendererInfo
  &NiagaraDataInterfaceSkeletalMesh
 ]
 usedDynamicInputs [&ClampFloat]
 usedModules [
  &self:SystemSpawnScript.SystemScriptSource.SystemScriptGraph.NiagaraNodeAssignment_0.SetVariables_#1
  &self:SystemSpawnScript.SystemScriptSource.SystemScriptGraph.NiagaraNodeAssignment_1.SetVariables_#2
  &self:SystemSpawnScript.SystemScriptSource.SystemScriptGraph.NiagaraNodeAssignment_5.SetVariables_#3
  &SystemState
  &self:SystemSpawnScript.SystemScriptSource.SystemScriptGraph.NiagaraNodeAssignment_2.SetVariables_#4
  &InterpolateOverTime
  &self:SystemSpawnScript.SystemScriptSource.SystemScriptGraph.NiagaraNodeAssignment_3.SetVariables_#5
  &self:SystemSpawnScript.SystemScriptSource.SystemScriptGraph.NiagaraNodeAssignment_4.SetVariables_#6
 ]
}
emitters {
 GroundDust {
  emitterUpdateScript [
   &EmitterState(Life Cycle Mode=::NewEnumerator1, Loop Behavior=::NewEnumerator1, Loop Duration=$System.CurrentLoopDuration, UseLoopDelay=false)
   &SpawnBurst_Instantaneous(Spawn Count=14, Use Spawn Probability=false, UseLoopCountLimit=false)
   /SetVariables(Emitter.SizeRad=150) {isSetParameters=true}
  ]
  particleSpawnScript [
   /SetVariables(Particles.DustColor:LinearColor=&LerpLinearColors(
     StartColor=&RandomRangeLinearColor(Minimum=<rgba 0.1146 0.1041 0.0995 0.5>, Maximum=<rgba 0.3229 0.2978 0.2778 0.5>)
     EndColor=&RandomRangeLinearColor(Minimum=<rgba 0.2083 0.0717 0.032 0.75>, Maximum=<rgba 0.4323 0.2135 0.144 0.75>)
     LerpFactor=&MakeCustomFloatFromBool(Bool=$User.DesertBoulder, True Float=1, False Float=0)
    )) {isSetParameters=true}
   /SetVariables(Particles.SizeMax=150, Particles.SizeMin=80) {isSetParameters=true}
   /SetVariables(Particles.SizeCurve:NiagaraFloat=&Lerp_Float(
     A=$Particles.SizeMin
     B=$Particles.SizeMax
     Alpha=&FloatFromCurve(
      FloatCurve=DataInterface() {propertyValues {Curve {keys [
          {interpMode=RCIM_Cubic, tangentMode=RCTM_Break, time=0, value=0}
          {arriveTangentWeight=0.9217, interpMode=RCIM_Cubic, tangentMode=RCTM_Break, time=1, value=1}
         ]}}}
      CurveIndex=&RandomRangeFloat(Minimum=0, Maximum=1)
     )
    )) {isSetParameters=true}
   /SetVariables(Particles.InitSpriteRotRate=&UniformRangedFloat(Minimum=-1, Maximum=1)) {isSetParameters=true}
   &InitializeParticle(
    Color=$Particles.DustColor
    Lifetime=&Lerp_Float(
     A=0.75
     B=1.2
     Alpha=&FloatFromCurve(
      FloatCurve=DataInterface() {propertyValues {Curve {keys [
          {interpMode=RCIM_Cubic, tangentMode=RCTM_User, time=0, value=0}
          {arriveTangent=2.8717, interpMode=RCIM_Cubic, leaveTangent=2.8717, tangentMode=RCTM_User, time=1, value=1}
         ]}}}
      CurveIndex=&RandomRangeFloat(Minimum=0, Maximum=1)
     )
    )
    Sprite Rotation Mode=::NewEnumerator2
    Sprite Size Mode=::NewEnumerator3
    Uniform Sprite Size=$Particles.SizeCurve
    UsePositionOffset=false
    Write Material Random=true
   )
   &ShapeLocation(Handle Radius=1, Large Radius=3, MaskSpawnGroup=false, Override Seed=false, Shape Primitive=::NewEnumerator3, UseOwnerScale=true)
   &AddVelocity(
    Module.Use Falloff Distance=false
    Origin Offset=<0 0 -0.3>
    Override Seed=false
    Velocity Mode=::NewEnumerator1
    Velocity Speed=&RandomRangeFloat(Minimum=150, Maximum=250)
   )
   &CameraOffset(Camera Offset Amount=&RandomRangeFloat(Minimum=15, Maximum=25))
  ]
  particleUpdateScript [
   &ParticleState()
   &Drag(Drag=4, Use Linear Drag=true, Use Rotational Drag=false)
   &AccelerationForce(Acceleration=&RandomRangeVector(Minimum=<-10 -10 300>, Maximum=<10 10 100>))
   &GravityForce(Gravity=&VectorFromCurve(
     VectorCurve=DataInterface() {propertyValues {
       XCurve {keys [<at 0 1>,<at 1 1>]}
       YCurve {keys [<at 0 1>,<at 1 1>]}
       ZCurve {keys [
         {interpMode=RCIM_Cubic, tangentMode=RCTM_User, time=0, value=0}
         {arriveTangent=0.6985, interpMode=RCIM_Cubic, leaveTangent=0.6985, tangentMode=RCTM_User, time=1, value=1}
        ]}
      }}
     Scale Curve=<0 0 -480>
    ))
   &Drag(
    Drag=&FloatFromCurve(
     FloatCurve=DataInterface() {propertyValues {Curve {keys [{interpMode=RCIM_Cubic, tangentMode=RCTM_User, time=0, value=0}, {interpMode=RCIM_Cubic, tangentMode=RCTM_User, time=1, value=1}]}}}
     Scale Curve=7
    )
    Use Linear Drag=true
    Use Rotational Drag=false
   ) {module=Drag001}
   &SolveForcesAndVelocity(Clamp Velocity=false, Limit Acceleration=false)
   &SpriteRotationRate(Rotation Rate=&Multiply_Float(
     A=$Particles.InitSpriteRotRate
     B=&FloatFromCurve(
      FloatCurve=DataInterface() {propertyValues {Curve {keys [{interpMode=RCIM_Cubic, tangentMode=RCTM_User, time=0, value=0}, {interpMode=RCIM_Cubic, tangentMode=RCTM_User, time=1, value=1}]}}}
      Scale Curve=&RandomRangeFloat(Minimum=-1, Maximum=1)
     )
    ))
   &ScaleSpriteSize(
    Scale Sprite Size Mode=::NewEnumerator0
    Uniform Scale Factor=&FloatFromCurve(FloatCurve=DataInterface() {propertyValues {Curve {keys [
         {interpMode=RCIM_Cubic, leaveTangent=3.1262, tangentMode=RCTM_Break, time=0, value=0}
         {arriveTangent=0.5926, interpMode=RCIM_Cubic, tangentMode=RCTM_Break, time=0.95, value=1}
         <at 1 0>
        ]}}})
   )
   &DynamicMaterialParameters(
    Index 0 Param 1=&Add_Float(
     A=&RandomRangeFloat(Minimum=0, Maximum=1)
     B=&FloatFromCurve(
      FloatCurve=DataInterface() {propertyValues {Curve {keys [{tangentMode=RCTM_User, time=0, value=0}, {tangentMode=RCTM_User, time=1, value=1}]}}}
      CurveIndex=&Add_Float(A=&Multiply_Float(A=$Particles.Initial.MaterialRandom, B=1), B=$Particles.NormalizedAge)
      Scale Curve=0.1
     )
    )
    Index 0 Param 2=&FloatFromCurve(FloatCurve=DataInterface() {propertyValues {Curve {keys [<at 0 1>,<at 1 0>]}}})
    Index 0 Param 3=&FloatFromCurve(
     FloatCurve=DataInterface() {propertyValues {Curve {keys [
         {interpMode=RCIM_Cubic, leaveTangent=16.5652, leaveTangentWeight=0.4385, tangentMode=RCTM_Break, time=0, value=0}
         {arriveTangentWeight=0.0333, interpMode=RCIM_Cubic, leaveTangentWeight=0.9477, tangentMode=RCTM_Break, time=0.1, value=1}
         {arriveTangent=-1.7159, interpMode=RCIM_Cubic, leaveTangent=-1.7159, tangentMode=RCTM_User, time=0.3, value=0.2}
         {arriveTangent=0.2347, interpMode=RCIM_Cubic, leaveTangent=0.2723, tangentMode=RCTM_Break, time=0.6, value=0}
        ]}}}
     Scale Curve=0
    )
    Index 0 Param 4=&FloatFromCurve(FloatCurve=DataInterface() {propertyValues {Curve {keys [{interpMode=RCIM_Cubic, tangentMode=RCTM_Break, time=0.5, value=1}, {interpMode=RCIM_Cubic, tangentMode=RCTM_Break, time=1, value=0}]}}})
    Index 1 Param 1=&FloatFromCurve(FloatCurve=DataInterface() {propertyValues {Curve {keys [
         {interpMode=RCIM_Cubic, leaveTangentWeight=0.0833, tangentMode=RCTM_Break, time=0, value=0}
         {arriveTangentWeight=0.0873, interpMode=RCIM_Cubic, tangentMode=RCTM_Break, time=0.25, value=1}
        ]}}})
    Index 1 Param 2=&FloatFromCurve(
     FloatCurve=DataInterface() {propertyValues {Curve {keys [
         {interpMode=RCIM_Cubic, tangentMode=RCTM_Break, time=0, value=0.2}
         {interpMode=RCIM_Cubic, leaveTangent=0.9066, tangentMode=RCTM_Break, time=1, value=2}
        ]}}}
     Scale Curve=&RandomRangeFloat(Minimum=0.1, Maximum=0.2)
    )
    Index 1 Param 3=&FloatFromCurve(FloatCurve=DataInterface() {propertyValues {Curve {keys [
         {interpMode=RCIM_Cubic, leaveTangent=6.3917, tangentMode=RCTM_Break, time=0, value=0}
         {arriveTangent=1.7549, interpMode=RCIM_Cubic, leaveTangent=10.6767, tangentMode=RCTM_User, time=0.25, value=1}
        ]}}})
    Index 1 Param 4=&FloatFromCurve(
     FloatCurve=DataInterface() {propertyValues {Curve {keys [{arriveTangent=-0.136, interpMode=RCIM_Cubic, leaveTangent=-0.136, tangentMode=RCTM_User, time=0, value=1}, <at 1 0.95>]}}}
     Scale Curve=2
    )
    Index 2 Param 1=&FloatFromCurve(
     FloatCurve=DataInterface() {propertyValues {Curve {keys [
         {interpMode=RCIM_Cubic, leaveTangent=4.2485, tangentMode=RCTM_Break, time=0, value=0}
         {arriveTangent=1.0548, interpMode=RCIM_Cubic, leaveTangent=1.0548, tangentMode=RCTM_User, time=0.17, value=0.5715}
         <at 1 1>
        ]}}}
     Scale Curve=3
    )
    Index 2 Param 2=&FloatFromCurve(
     FloatCurve=DataInterface() {propertyValues {Curve {keys [{interpMode=RCIM_Cubic, tangentMode=RCTM_User, time=0, value=0}, {interpMode=RCIM_Cubic, tangentMode=RCTM_User, time=1, value=1}]}}}
     Scale Curve=0.3
    )
    Index 2 Param 3=&FloatFromCurve(
     FloatCurve=DataInterface() {propertyValues {Curve {keys [
         {interpMode=RCIM_Cubic, leaveTangent=-15.7052, leaveTangentWeight=0.1411, tangentMode=RCTM_Break, time=0, value=1}
         {interpMode=RCIM_Cubic, tangentMode=RCTM_Break, time=0.1, value=0}
         {arriveTangent=0.7532, interpMode=RCIM_Cubic, leaveTangent=0.7532, tangentMode=RCTM_User, time=0.7, value=0.5}
         {arriveTangent=2.358, arriveTangentWeight=1.057, interpMode=RCIM_Cubic, tangentMode=RCTM_Break, time=1, value=1}
        ]}}}
     Scale Curve=5
    )
    Index 2 Param 4=&RandomRangeFloat(Minimum=1.5, Maximum=1.75)
    Module.Param0WriteEnabled=true
    Module.Param0WriteEnabled1=true
    Module.Param0WriteEnabled2=true
    Module.Param1WriteEnabled=false
    Module.Param1WriteEnabled1=true
    Module.Param1WriteEnabled2=true
    Module.Param2WriteEnabled=true
    Module.Param2WriteEnabled1=true
    Module.Param2WriteEnabled2=true
    Module.Param3WriteEnabled=true
    Module.Param3WriteEnabled1=true
    Module.Param3WriteEnabled2=true
    Write Parameter Index 1=true
    Write Parameter Index 2=true
   )
   &ScaleColor(
    Scale Alpha=&FloatFromCurve(
     FloatCurve=DataInterface() {propertyValues {Curve {keys [
         {arriveTangent=6.2714, interpMode=RCIM_Cubic, leaveTangent=6.2714, tangentMode=RCTM_User, time=0, value=0}
         {arriveTangent=8.4935, interpMode=RCIM_Cubic, leaveTangent=8.4935, tangentMode=RCTM_User, time=0.1, value=1}
         {arriveTangent=0.3743, interpMode=RCIM_Cubic, leaveTangent=0.3743, tangentMode=RCTM_User, time=0.2, value=1.5}
         {arriveTangent=-0.0578, interpMode=RCIM_Cubic, leaveTangent=-0.0578, tangentMode=RCTM_User, time=0.3, value=1}
         <at 1 1>
        ]}}}
     Scale Curve=&Multiply_Float(
      A=&RandomRangeFloat(Minimum=1.5, Maximum=2)
      B=&FNTimeOfDayScalar(Night Scale=2, Morning Scale=1, Evening Scale=2)
     )
    )
    ScaleA=true
    ScaleRGB=false
   )
  ]
  rendererClasses [&NiagaraSpriteRendererProperties]
  renderers [
   {rendererClass=&NiagaraSpriteRendererProperties}
   {rendererClass=&NiagaraSpriteRendererProperties, rendererIndex=1}
  ]
 }
 MainBurst {
  emitterSpawnScript [
   /SetVariables(
    Emitter.Length=85
    Emitter.ParticlesToSpawn=12
    Emitter.Random:NiagaraFloat=&RandomRangeFloat(Minimum=0, Maximum=1)
   ) {isSetParameters=true}
   /SetVariables(
    Emitter.SpriteLength=70
    Emitter.SpriteRandom:NiagaraFloat=&RandomRangeFloat(Minimum=0, Maximum=1)
    Emitter.SpritesToSpawn=3
   ) {isSetParameters=true}
  ]
  emitterUpdateScript [
   &EmitterState(Life Cycle Mode=::NewEnumerator1, Loop Behavior=::NewEnumerator1, Loop Delay=0.2, Loop Duration=1, UseLoopDelay=false)
   &SpawnBurst_Instantaneous(Spawn Count=$Emitter.ParticlesToSpawn, Spawn Time=0.2, Use Spawn Probability=false, UseLoopCountLimit=false)
   &SpawnBurst_Instantaneous(Spawn Count=$Emitter.SpritesToSpawn, Spawn Probability=0.5, Use Spawn Probability=false, UseLoopCountLimit=false) {module=SpawnBurst_Instantaneous001}
   &SpawnBurst_Instantaneous(
    Spawn Count=&RandomRangeInt(Minimum=8, Maximum=12)
    Spawn Group=1
    Spawn Time=0.03
    Use Spawn Probability=false
    UseLoopCountLimit=false
   ) {module=SpawnBurst_Instantaneous002}
   /SetVariables(Emitter.Reveal:NiagaraFloat=&Multiply_Float(
     A=&Lerp_Float(
      A=&FloatFromCurve(FloatCurve=DataInterface() {propertyValues {Curve {keys [<at 0 0>,<at 0.6 1>]}}}, CurveIndex=$User.Reveal)
      B=&FloatFromCurve(FloatCurve=DataInterface() {propertyValues {Curve {keys [<at 0 0>,<at 0.95 1>]}}}, CurveIndex=$User.Reveal)
      Alpha=$System.SystemRandom
     )
     B=&Lerp_Float(A=1, B=1.1, Alpha=$Emitter.Random)
    )) {isSetParameters=true}
   /SetVariables(Emitter.Length:NiagaraFloat=&Multiply_Float(
     A=&FloatFromCurve(
      FloatCurve=DataInterface() {propertyValues {Curve {keys [
          {arriveTangent=2.5813, interpMode=RCIM_Cubic, leaveTangent=2.5813, tangentMode=RCTM_User, time=0, value=0.1}
          {tangentMode=RCTM_User, time=0.5, value=1}
         ]}}}
      CurveIndex=$Engine.System.Age
     )
     B=&Multiply_Float(
      A=$Emitter.Initial.Length
      B=&FloatFromCurve(
       FloatCurve=DataInterface() {propertyValues {Curve {keys [
           {interpMode=RCIM_Cubic, leaveTangent=6.4155, tangentMode=RCTM_Break, time=0, value=0}
           {arriveTangent=0.1191, interpMode=RCIM_Cubic, leaveTangent=0.1191, tangentMode=RCTM_User, time=0.1, value=0.9}
           {arriveTangent=0.1177, interpMode=RCIM_Cubic, leaveTangent=0.1177, tangentMode=RCTM_User, time=1, value=1}
          ]}}}
       CurveIndex=$Emitter.Reveal
       Scale Curve=&Lerp_Float(A=0.7, B=0.8, Alpha=$System.SizeRandom)
      )
     )
    )) {isSetParameters=true}
   /SetVariables(Emitter.SpriteReveal:NiagaraFloat=&Multiply_Float(
     A=&Lerp_Float(
      A=&FloatFromCurve(FloatCurve=DataInterface() {propertyValues {Curve {keys [<at 0 0>,<at 0.6 1>]}}}, CurveIndex=$User.Reveal)
      B=&FloatFromCurve(FloatCurve=DataInterface() {propertyValues {Curve {keys [<at 0 0>,<at 0.95 1>]}}}, CurveIndex=$User.Reveal)
      Alpha=$System.SystemRandom
     )
     B=&Lerp_Float(A=1, B=1.1, Alpha=$Emitter.Random)
    )) {isSetParameters=true}
   /SetVariables(Emitter.SpriteLength:NiagaraFloat=&Multiply_Float(
     A=&FloatFromCurve(
      FloatCurve=DataInterface() {propertyValues {Curve {keys [
          {arriveTangent=2.5813, interpMode=RCIM_Cubic, leaveTangent=2.5813, tangentMode=RCTM_User, time=0, value=0.1}
          {tangentMode=RCTM_User, time=0.5, value=1}
         ]}}}
      CurveIndex=$Engine.System.Age
     )
     B=&Multiply_Float(
      A=$Emitter.Initial.SpriteLength
      B=&FloatFromCurve(
       FloatCurve=DataInterface() {propertyValues {Curve {keys [
           {interpMode=RCIM_Cubic, leaveTangent=6.4155, tangentMode=RCTM_Break, time=0, value=0}
           {arriveTangent=0.1191, interpMode=RCIM_Cubic, leaveTangent=0.1191, tangentMode=RCTM_User, time=0.1, value=0.9}
           {arriveTangent=0.1177, interpMode=RCIM_Cubic, leaveTangent=0.1177, tangentMode=RCTM_User, time=1, value=1}
          ]}}}
       CurveIndex=$Emitter.SpriteReveal
       Scale Curve=&Lerp_Float(A=0.7, B=1, Alpha=$System.SizeRandom)
      )
     )
    )) {isSetParameters=true}
  ]
  particleSpawnScript [
   &InitializeParticle(
    Color=&MaskLinearColorBySpawnGroup(Masked Linear Color=<rgba 1 1 1 1>, Passthrough Linear Color=<rgba 1 1 1 0.6>)
    Lifetime=&MaskFloatBySpawnGroup(Masked Float=$System.CurrentLoopDuration, Passthrough Float=&UniformRangedFloat(Minimum=0.4, Maximum=0.6))
    Position Offset=&MaskVectorBySpawnGroup(Masked Vector=<0 0 0>, Passthrough Vector=<0 -5 0>)
    Sprite Rotation Mode=::NewEnumerator2
    Sprite Size Mode=::NewEnumerator3
    Sprite UV Mode=::NewEnumerator4
    Uniform Sprite Size=&MaskFloatBySpawnGroup(Masked Float=10, Passthrough Float=&UniformRangedFloat(Minimum=10, Maximum=15))
    UsePositionOffset=true
    Write Material Random=true
   )
   &DynamicMaterialParameters(
    Index 0 Param 1=$Emitter.Random
    Module.Index0Param0Write=true
    Module.Index0Param1Write=true
    Module.Index0Param2Write=true
    Module.Index0Param3Write=true
   )
   /SetVariables(Particles.VisibilityTag=&SpawnGroup()) {isSetParameters=true}
   &ShapeLocation(
    MaskSpawnGroup=false
    Offset=<0 10 15>
    Offset Mode=::NewEnumerator0
    Override Seed=false
    Sphere Radius=&MaskFloatBySpawnGroup(Masked Float=0, Passthrough Float=0.1)
    UseOwnerScale=true
   ) {module=ShapeLocation001}
   &ApplyInitialForces()
   &AddVelocity(
    Cone Angle=5
    Cone Axis=<1 0.9 0>
    Module.UseConeFalloff=false
    Module.UseVelDistribution=false
    Override Seed=false
    Velocity Mode=::NewEnumerator2
    Velocity Speed=&MaskFloatBySpawnGroup(
     Masked Float=0
     Passthrough Float=&Multiply_Float(
      A=&RandomRangeFloat(Minimum=2, Maximum=250)
      B=&Lerp_Float(A=0.7, B=1.2, Alpha=$System.SizeRandom)
     )
    )
   )
   &InitialMeshOrientation(
    Mesh Axis To Orient=<0 1 0>
    Mesh Orientation Mode=::NewEnumerator2
    Orientation Vector=$Particles.ShapeLocation001.ShapePosition
    Rotation=<44.6186 137.0103 0>
    Use Rotation Vector=true
   )
  ]
  particleUpdateScript [
   &SolveForcesAndVelocity(Clamp Velocity=false, Limit Acceleration=false)
   &ParticleState()
   &SkeletalMeshLocation(Bone Sampling Mode=::NewEnumerator4, MaskSpawnGroup=false, Position Sampling=::NewEnumerator2, Skeletal Mesh=$System.Skeletal Mesh)
   &UpdateMeshOrientation(
    Facing Direction:NiagaraPosition=$Particles.ShapeLocation001.ShapeNormal
    Facing Interpolation=false
    Orientation Method=::NewEnumerator1
    Override Quat=false
    TransformFacing=false
    TransformUp=false
   )
   /SetVariables(Particles.ParticleRelativeLifetimeRibbons:NiagaraFloat=&ReturnNormalizedExecIndex()) {isSetParameters=true}
   /SetVariables(Particles.OpacityMultiplierRibbons:NiagaraFloat=&FloatFromCurve(
     FloatCurve=DataInterface() {propertyValues {Curve {keys [
         {interpMode=RCIM_Cubic, tangentMode=RCTM_User, time=0, value=0}
         {arriveTangent=0.007, interpMode=RCIM_Cubic, leaveTangent=0.007, tangentMode=RCTM_User, time=0.5, value=1.5}
         {arriveTangent=-0.7513, interpMode=RCIM_Cubic, leaveTangent=-0.7513, tangentMode=RCTM_User, time=0.9, value=1}
         {arriveTangent=-0.9808, interpMode=RCIM_Cubic, leaveTangent=-0.9808, tangentMode=RCTM_User, time=1.8, value=0}
        ]}}}
     CurveIndex=&Add_Float(
      A=&Multiply_Float(A=$Particles.ParticleRelativeLifetimeRibbons, B=-1)
      B=&Multiply_Float(A=$Emitter.Reveal, B=2)
     )
    )) {isSetParameters=true}
   /SetVariables(Particles.InterpolateCurveRibbons:NiagaraFloat=&FloatFromCurve(
     FloatCurve=DataInterface() {propertyValues {Curve {keys [{arriveTangent=-0.6503, time=0, value=0}, {arriveTangent=0.8679, time=1, value=1}]}}}
     CurveIndex=$Particles.ParticleRelativeLifetimeRibbons
    )) {isSetParameters=true}
   /SetVariables(Particles.PositionRibbons:NiagaraPosition=&RotateVectorByQuaternion(
     Vector To Rotate=&AddVectorToPosition(
      Position=&AddVectorToPosition(
       Position=$Particles.ShapeLocation001.ShapePosition
       Vector=&TransformVector(Vector=<0 -5 0>, Destination Space=::Simulation)
      )
      Vector=&VectorFromFloat(Value=&Abs_Float(Float:NiagaraFloat=&NormalizeFloat(Value=&Multiply_Float(A=$Emitter.Length, B=$Particles.InterpolateCurveRibbons), Min=-1, Max=1)))
     )
     Quaternion=$System.Foot_Rotator
    )) {isSetParameters=true}
   /SetVariables(
    Particles.RibbonColor:LinearColor=&MultiplyLinearColor(
     Linear Color=&MultiplyLinearColor(
      Linear Color=&ColorFromCurve(
       ColorCurve=DataInterface() {propertyValues {
         AlphaCurve {keys [
           <at 0 0>
           <at 0.0346 1>
           <at 0.2133 0.8>
           {arriveTangent=-1.0175, interpMode=RCIM_Cubic, leaveTangent=-1.0175, time=0.7268, value=0.8}
           <at 0.9995 0 RCIM_Cubic>
          ]}
         BlueCurve {keys [<at 0.4438 1>]}
         GreenCurve {keys [<at 0.4438 1>]}
         RedCurve {keys [<at 0.4438 1>]}
        }}
       CurveIndex=$Particles.ParticleRelativeLifetimeRibbons
      )
      Scale Factor=&MakeVector4_FromLinearColor(LinearColor=&FNTimeOfDayColor(Night Color=<rgba 0.1771 0.2756 1 0.7>, Morning Color=<rgba 1 1 1 1>, Day Color=<rgba 1 1 1 1>, Evening Color=<rgba 1 0.8696 0.5989 1>))
     )
     Scale Factor=&MakeFloat4FromVectorAndFloat(
      "Vector (RGB)"=&VectorFromFloat(Value=&FloatFromCurve(FloatCurve=DataInterface() {propertyValues {Curve {keys [<at 0.45 1 RCIM_Cubic>]}}}, CurveIndex=$User.Reveal))
      "Float (Alpha)"=1
     )
    )
    Particles.RibbonWidth=&Multiply_Float(
     A=&FloatFromCurve(
      FloatCurve=DataInterface() {propertyValues {Curve {keys [
          {arriveTangent=0.675, interpMode=RCIM_Cubic, leaveTangent=0.675, tangentMode=RCTM_User, time=0, value=0.2}
          {arriveTangent=1.0523, interpMode=RCIM_Cubic, leaveTangent=1.0523, tangentMode=RCTM_User, time=0.2799, value=0.5408}
          {arriveTangent=-0.0372, interpMode=RCIM_Cubic, leaveTangent=-0.0372, tangentMode=RCTM_User, time=0.7, value=0.8}
          {arriveTangent=-1.5957, interpMode=RCIM_Cubic, leaveTangent=-1.5957, tangentMode=RCTM_User, time=1, value=0.5138}
         ]}}}
      CurveIndex=$Particles.InterpolateCurveRibbons
      Scale Curve=15
     )
     B=&FloatFromCurve(
      FloatCurve=DataInterface() {propertyValues {Curve {keys [
          {arriveTangent=5.9942, interpMode=RCIM_Cubic, leaveTangent=5.9942, tangentMode=RCTM_User, time=0, value=0.475}
          {arriveTangent=0.0898, interpMode=RCIM_Cubic, leaveTangent=0.0898, tangentMode=RCTM_User, time=0.3, value=1}
         ]}}}
      CurveIndex=$System.Age
     )
    )
   ) {isSetParameters=true}
   /SetVariables(Particles.ScaledRibbonWidth:NiagaraFloat=&Multiply_Float(
     A=$Particles.RibbonWidth
     B=&Multiply_Float(
      A=&FloatFromCurve(
       FloatCurve=DataInterface() {propertyValues {Curve {keys [
           {interpMode=RCIM_Cubic, leaveTangent=0.5491, tangentMode=RCTM_Break, time=0, value=0.7}
           {arriveTangent=1.0086, interpMode=RCIM_Cubic, leaveTangent=1.0086, tangentMode=RCTM_User, time=0.3931, value=1}
           {arriveTangent=0.0216, interpMode=RCIM_Cubic, tangentMode=RCTM_Break, time=1, value=1.5}
          ]}}}
       CurveIndex=$Emitter.Reveal
       Scale Curve=&Lerp_Float(
        A=1.6
        B=&FloatFromCurve(
         FloatCurve=DataInterface() {propertyValues {Curve {keys [<at 0 1.9 RCIM_Constant>,<at 0.333 2.25 RCIM_Constant>,<at 0.666 2.6 RCIM_Constant>]}}}
         CurveIndex=$Emitter.Random
        )
        Alpha=$System.SizeRandom
       )
      )
      B=1
     )
    )) {isSetParameters=true}
   /SetVariables(Particles.ScaledRibbonColor:LinearColor=&Multiply_LinearColorByFloat(
     Linear Color=&MultiplyLinearColor(
      Linear Color=$Particles.RibbonColor
      Scale Factor=&MakeFloat4FromVectorAndFloat(
       "Vector (RGB)"=<1 1 1>
       "Float (Alpha)"=&Multiply_Float(
        A=$Particles.OpacityMultiplierRibbons
        B=&Multiply_Float(
         A=&FloatFromCurve(
          FloatCurve=DataInterface() {propertyValues {Curve {keys [
              {interpMode=RCIM_Cubic, tangentMode=RCTM_Break, time=0.6, value=1}
              {arriveTangent=-4.2498, interpMode=RCIM_Cubic, tangentMode=RCTM_Break, time=0.99, value=0}
             ]}}}
          CurveIndex=$Emitter.Reveal
          Scale Curve=&Lerp_Float(A=0.8, B=1, Alpha=$Emitter.Random)
         )
         B=$System.OpacityRandom
        )
       )
      )
     )
     Float=&FloatFromCurve(
      FloatCurve=DataInterface() {propertyValues {Curve {keys [
          {arriveTangent=5.7153, interpMode=RCIM_Cubic, leaveTangent=5.7153, tangentMode=RCTM_User, time=0, value=0}
          {arriveTangent=-1.4405, interpMode=RCIM_Cubic, leaveTangent=-1.4405, tangentMode=RCTM_User, time=0.3, value=2}
          {arriveTangent=-1.4711, interpMode=RCIM_Cubic, leaveTangent=-1.4711, tangentMode=RCTM_User, time=1, value=0}
         ]}}}
      CurveIndex=$System.Age
     )
    )) {isSetParameters=true}
   &CameraOffset(Camera Offset Amount=5) {enabled=false}
   &ParticleState() {module=ParticleState001}
   &ScaleSpriteSize(Scale Factor=&MaskVector2dBySpawnGroup(
     Masked Vector2D=$Particles.Initial.SpriteSize
     Passthrough Vector2D=&Multiply_Vector2D_ByFloat(
      Float=&FloatFromCurve(
       FloatCurve=DataInterface() {propertyValues {Curve {keys [
           {arriveTangent=5.7302, interpMode=RCIM_Cubic, leaveTangent=5.7302, tangentMode=RCTM_User, time=0, value=0}
           {arriveTangent=0.2869, interpMode=RCIM_Cubic, leaveTangent=0.2869, tangentMode=RCTM_User, time=0.5, value=0.975}
          ]}}}
       CurveIndex=$Engine.System.Age
      )
      Vector2D=&Vector2DFromFloat(Value=&FloatFromCurve(
        FloatCurve=DataInterface() {propertyValues {Curve {keys [{tangentMode=RCTM_User, time=0, value=0.4}, <at 0.1 0.1>, {tangentMode=RCTM_User, time=1, value=1}]}}}
        Scale Curve=&Multiply_Float(A=0.7, B=&Lerp_Float(A=0.7, B=1, Alpha=$System.SizeRandom))
       ))
     )
     Spawn Group Mask=0
    )) {module=ScaleSpriteSize001}
   &SpriteRotationRate(Rotation Rate=&Multiply_Float(
     A=&RandomRangeFloat(Minimum=100, Maximum=150)
     B=&UniformAOrBFloat(A=-1, Evaluation Type=::NewEnumerator0, Recalculate Random Each Loop=true, Override Seed=false)
    ))
   /SetVariables(Particles.ParticleRelativeLifetimeSprites:NiagaraFloat=&Frac_Float(Float=&Add_Float(A=&Multiply_Float(A=1.5, B=$System.Age), B=&ReturnNormalizedExecIndex()))) {isSetParameters=true}
   /SetVariables(Particles.OpacityMultiplierSprites:NiagaraFloat=&FloatFromCurve(
     FloatCurve=DataInterface() {propertyValues {Curve {keys [
         {interpMode=RCIM_Cubic, tangentMode=RCTM_User, time=0, value=0}
         {arriveTangent=-0.6677, interpMode=RCIM_Cubic, leaveTangent=-0.6677, tangentMode=RCTM_User, time=0.2, value=1.5}
         {arriveTangent=-0.6675, tangentMode=RCTM_User, time=1.5, value=0}
        ]}}}
     CurveIndex=&Add_Float(
      A=&Multiply_Float(A=$Particles.ParticleRelativeLifetimeSprites, B=-1)
      B=&Multiply_Float(A=$Emitter.SpriteReveal, B=&RandomRangeFloat(Minimum=1, Maximum=2))
     )
    )) {isSetParameters=true}
   /SetVariables(Particles.InterpolateCurveSprites:NiagaraFloat=&FloatFromCurve(
     FloatCurve=DataInterface() {propertyValues {Curve {keys [
         {arriveTangent=-0.6503, interpMode=RCIM_Cubic, leaveTangent=1.4736, tangentMode=RCTM_Break, time=0, value=0}
         {arriveTangent=0.4202, interpMode=RCIM_Cubic, tangentMode=RCTM_Break, time=1, value=1}
        ]}}}
     CurveIndex=$Particles.ParticleRelativeLifetimeSprites
    )) {isSetParameters=true}
   /SetVariables(Particles.PositionSprites:NiagaraPosition=&RotateVectorByQuaternion(
     Vector To Rotate=&AddVectorToPosition(
      Position=&AddVectorToPosition(
       Position=Linked() {linkedVariable {name=Particles.ShapeLocation001.ShapePosition, type=Vector3f}}
       Vector=&TransformVector(Vector=<0 -5 0>, Destination Space=::Simulation)
      )
      Vector=&VectorFromFloat(Value=&Abs_Float(Float:NiagaraFloat=&NormalizeFloat(Value=&Multiply_Float(A=$Emitter.SpriteLength, B=$Particles.InterpolateCurveSprites), Min=-1, Max=1)))
     )
     Quaternion=$System.Foot_Rotator
    )) {isSetParameters=true}
   /SetVariables(Particles.SpriteSize=&MaskVector2dBySpawnGroup(
     Masked Vector2D=&Multiply_Vector2D_ByFloat(
      Float=&FloatFromCurve(
       FloatCurve=DataInterface() {propertyValues {Curve {keys [
           {arriveTangent=4.522, interpMode=RCIM_Cubic, leaveTangent=4.522, tangentMode=RCTM_User, time=0, value=0.475}
           {arriveTangent=0.0898, interpMode=RCIM_Cubic, leaveTangent=0.0898, tangentMode=RCTM_User, time=0.3, value=1}
          ]}}}
       CurveIndex=$System.Age
      )
      Vector2D=&Multiply_Vector2D(
       A=&Vector2DFromFloat(Value=&FloatFromCurve(
         FloatCurve=DataInterface() {propertyValues {Curve {keys [{tangentMode=RCTM_User, time=0, value=0}, {tangentMode=RCTM_User, time=1, value=1}]}}}
         CurveIndex=$Particles.InterpolateCurveSprites
         Scale Curve=&Lerp_Float(A=35, B=50, Alpha=&RandomRangeFloat(Minimum=0, Maximum=1))
        ))
       B=&Vector2DFromFloat(Value=&Lerp_Float(A=0.7, B=1, Alpha=$System.SizeRandom))
      )
     )
     Passthrough Vector2D=$Particles.Initial.SpriteSize
     Spawn Group Mask=0
    )) {isSetParameters=true}
   /SetVariables(Particles.Position=&RotateVectorByQuaternion(Vector To Rotate=$Particles.ShapeLocation001.ShapePosition, Quaternion=$System.Foot_Rotator)) {isSetParameters=true}
   &SubUVAnimation(Start Frame=20, SubUV Lookup Index:NiagaraFloat=$Particles.ParticleRelativeLifetimeSprites)
   &SubUVAnimation(
    Output Alternate Renderer Binding=true
    "Sprite Renderer "=DataInterface() {propertyValues {SpriteRenderer=&self:MainBurst_1.NiagaraSpriteRendererProperties_1}}
    Start Frame Range Override=40
    UseEndFrame=true
    UseStartFrame=true
   ) {module=SubUVAnimation002}
   &ScaleColor(
    Scale Alpha=&MaskFloatBySpawnGroup(
     Masked Float=&Multiply_Float(
      A=&Multiply_Float(
       A=&FloatFromCurve(
        FloatCurve=DataInterface() {propertyValues {Curve {keys [
            {arriveTangent=16.5612, interpMode=RCIM_Cubic, leaveTangentWeight=0.3333, time=0, value=0}
            {arriveTangentWeight=0.1333, interpMode=RCIM_Cubic, leaveTangentWeight=0.1667, tangentMode=RCTM_User, time=0.333, value=1}
            {arriveTangentWeight=0.1667, interpMode=RCIM_Cubic, tangentMode=RCTM_User, time=1, value=0}
           ]}}}
        CurveIndex=$Particles.ParticleRelativeLifetimeSprites
        Scale Curve=&RandomRangeFloat(Minimum=0.5, Maximum=1)
       )
       B=&Multiply_Float(
        A=&FloatFromCurve(
         FloatCurve=DataInterface() {propertyValues {Curve {keys [
             {interpMode=RCIM_Cubic, tangentMode=RCTM_Break, time=0.6, value=1}
             {arriveTangent=-4.2498, interpMode=RCIM_Cubic, tangentMode=RCTM_Break, time=0.99, value=0}
            ]}}}
         CurveIndex=$Emitter.SpriteReveal
         Scale Curve=&Lerp_Float(A=1, B=2, Alpha=$Emitter.Random)
        )
        B=$System.OpacityRandom
       )
      )
      B=$Particles.OpacityMultiplierSprites
     )
     Passthrough Float=&Multiply_Float(
      A=&FloatFromCurve(
       FloatCurve=DataInterface() {propertyValues {Curve {keys [
           {arriveTangent=5.5857, interpMode=RCIM_Cubic, leaveTangent=5.5857, leaveTangentWeight=0.3333, tangentMode=RCTM_User, time=0, value=0.3}
           {arriveTangentWeight=0.1333, interpMode=RCIM_Cubic, leaveTangentWeight=0.1667, tangentMode=RCTM_User, time=0.1, value=2}
           {arriveTangentWeight=0.1667, interpMode=RCIM_Cubic, tangentMode=RCTM_User, time=1, value=0}
          ]}}}
       Scale Curve=&Multiply_Float(A=1, B=$System.OpacityRandom)
      )
      B=&RandomRangeFloat(Minimum=0.5, Maximum=1)
     )
    )
    Scale RGB=&MakeVectorFromLinearColorRGB(Color=&FNTimeOfDayColor(
      Night Color=<rgba 0.1771 0.2756 1 0.7>
      Morning Color=<rgba 0.5607 0.6584 1 1>
      Day Color=<rgba 1 1 1 1>
      Evening Color=<rgba 1 0.8696 0.5989 1>
     ))
    ScaleA=true
    ScaleRGB=true
   )
   &ScaleColor(
    Scale Alpha=&MaskFloatBySpawnGroup(
     Masked Float=&Multiply_Float(
      A=$Particles.OpacityMultiplierSprites
      B=&FloatFromCurve(
       FloatCurve=DataInterface() {propertyValues {Curve {keys [
           {arriveTangent=9.0725, interpMode=RCIM_Cubic, leaveTangent=9.0725, tangentMode=RCTM_User, time=0, value=0}
           {arriveTangent=-0.6131, interpMode=RCIM_Cubic, leaveTangent=-0.6131, tangentMode=RCTM_User, time=0.1408, value=0.9847}
           {arriveTangent=-1.4711, interpMode=RCIM_Cubic, leaveTangent=-1.4711, tangentMode=RCTM_User, time=1, value=0}
          ]}}}
       CurveIndex=$Emitter.Age
      )
     )
     Passthrough Float=&FloatFromCurve(
      FloatCurve=DataInterface() {propertyValues {Curve {keys [
          {arriveTangent=9.0725, interpMode=RCIM_Cubic, leaveTangent=9.0725, tangentMode=RCTM_User, time=0, value=0}
          {arriveTangent=-2.5249, interpMode=RCIM_Cubic, leaveTangent=-2.5249, tangentMode=RCTM_User, time=0.1408, value=0.9847}
          {arriveTangent=-0.0405, interpMode=RCIM_Cubic, leaveTangent=-0.0405, tangentMode=RCTM_User, time=1, value=0}
         ]}}}
      CurveIndex=$Emitter.Age
     )
    )
    ScaleA=true
    ScaleRGB=false
   ) {module=ScaleColor001}
   &ScaleColor(
    Scale Alpha=&MaskFloatBySpawnGroup(
     Masked Float=&Multiply_Float(
      A=$Particles.OpacityMultiplierSprites
      B=&FloatFromCurve(
       FloatCurve=DataInterface() {propertyValues {Curve {keys [
           {arriveTangent=9.0725, interpMode=RCIM_Cubic, leaveTangent=9.0725, tangentMode=RCTM_User, time=0, value=0}
           {arriveTangent=-0.6131, interpMode=RCIM_Cubic, leaveTangent=-0.6131, tangentMode=RCTM_User, time=0.1408, value=0.9847}
           {arriveTangent=-1.4711, interpMode=RCIM_Cubic, leaveTangent=-1.4711, tangentMode=RCTM_User, time=1, value=0}
          ]}}}
       CurveIndex=$Emitter.Age
      )
     )
     Passthrough Float=&FloatFromCurve(
      FloatCurve=DataInterface() {propertyValues {Curve {keys [
          {arriveTangent=9.0725, interpMode=RCIM_Cubic, leaveTangent=9.0725, tangentMode=RCTM_User, time=0, value=0}
          {arriveTangent=-2.5249, interpMode=RCIM_Cubic, leaveTangent=-2.5249, tangentMode=RCTM_User, time=0.1408, value=0.9847}
          {arriveTangent=-0.0405, interpMode=RCIM_Cubic, leaveTangent=-0.0405, tangentMode=RCTM_User, time=1, value=0}
         ]}}}
      Scale Curve=&Multiply_Float(A=1, B=$System.ToD)
     )
    )
    ScaleA=true
    ScaleRGB=false
   ) {module=ScaleColor002}
  ]
  rendererClasses [&NiagaraRibbonRendererProperties, &NiagaraSpriteRendererProperties]
  renderers [
   {rendererClass=&NiagaraRibbonRendererProperties}
   {rendererClass=&NiagaraSpriteRendererProperties, rendererIndex=1}
   {rendererClass=&NiagaraSpriteRendererProperties, rendererIndex=2}
  ]
 }
 SmallBits {
  simTarget=GPUComputeSim
  emitterUpdateScript [
   &EmitterState()
   &SpawnBurst_Instantaneous(
    Spawn Count=&RandomRangeInt(Minimum=35, Maximum=55)
    Use Spawn Probability=false
    UseLoopCountLimit=false
   )
  ]
  particleSpawnScript [
   /SetVariables(Particles.randomInt:NiagaraInt32=&RandomRangeInt(Minimum=0, Maximum=1)) {isSetParameters=true}
   &InitializeParticle(
    Color Maximum=<rgba 0.526 0.5217 0.4789 5>
    Color Minimum=<rgba 0.2279 0.2292 0.2078 5>
    Color Mode=::NewEnumerator2
    Lifetime=&RandomRangeFloat(Minimum=1, Maximum=2)
    Mesh Scale Mode=::NewEnumerator1
    Mesh Uniform Scale Max=50
    Mesh Uniform Scale Min=12
    Sprite Rotation Mode=::NewEnumerator2
    Sprite Size Mode=::NewEnumerator1
    Uniform Sprite Size Max=5
    Uniform Sprite Size Min=1
    UsePositionOffset=false
    Write Material Random=true
   )
   &SubUVAnimation(End Frame=3, SubUV Animation Mode:ENiagara_SubUVLookupMode=::NewEnumerator2)
   &ShapeLocation(MaskSpawnGroup=false, Override Seed=false, Sphere Radius=10, UseOwnerScale=true)
   /SetVariables(Particles.InitSpriteRotRate=&UniformRangedFloat(Minimum=-3, Maximum=3)) {isSetParameters=true}
   &AddVelocity(
    Module.Use Falloff Distance=false
    Origin Offset=<0 0 -15>
    Override Seed=false
    Velocity Mode=::NewEnumerator1
    Velocity Speed=&RandomRangeFloat(Minimum=150, Maximum=1000)
   )
  ]
  particleUpdateScript [
   &ParticleState()
   &SpriteRotationRate(Rotation Rate=&Multiply_Float(
     A=$Particles.InitSpriteRotRate
     B=&FloatFromCurve(
      FloatCurve=DataInterface() {propertyValues {Curve {keys [
          {arriveTangent=-0.2085, interpMode=RCIM_Cubic, leaveTangent=-0.2085, tangentMode=RCTM_User, time=-0.0046, value=1.0113}
          {arriveTangent=-0.2496, interpMode=RCIM_Cubic, leaveTangent=-0.2496, tangentMode=RCTM_User, time=0.5, value=0.05}
         ]}}}
      CurveIndex=$Emitter.Age
      Scale Curve=950
     )
    ))
   &Collision(
    Age Colliding Particles=false
    Allow Offscreen Depth Buffer Collision Particles To Live After a Bounce=true
    Collision Enabled=&SetBoolByFloatComparison(A=$Emitter.Age, B=0.15)
    Control Roll On Collision=true
    Friction During a Bounce=0.2
    Kill Occluded Particles:NiagaraBool=&SetBoolByFloatComparison(A=$Emitter.Age, B=0.15)
    Kill Offscreen Particles=true
    Module.Control Rest State Via Surface Normal=false
    Module.EnableMaxCollisionCount=false
    Module.Randomize Collision Normal Vector=true
    Restitution=0.5
    ScaleRateOfRotation=false
   ) {enabled=false}
   &GravityForce(Gravity=&RandomRangeVector(Minimum=<0 0 -980>, Maximum=<1 1 -1500>))
   &Drag(
    Drag=&FloatFromCurve(
     FloatCurve=DataInterface() {propertyValues {Curve {keys [{arriveTangent=-1, interpMode=RCIM_Cubic, leaveTangent=-3.4078, tangentMode=RCTM_Break, time=0, value=1}, <at 0.3 0.1>, <at 1 0>]}}}
     Scale Curve=7
    )
    Use Linear Drag=true
    Use Rotational Drag=false
   )
   &SolveForcesAndVelocity(Clamp Velocity=false, Limit Acceleration=false)
   &ScaleSpriteSize(
    Uniform Curve Scale=2.5
    Uniform Curve Sprite Scale=DataInterface() {propertyValues {Curve {keys [
        {arriveTangent=40, interpMode=RCIM_Cubic, leaveTangent=40, time=0, value=0}
        {arriveTangent=-2.1262, interpMode=RCIM_Cubic, leaveTangent=-2.1262, tangentMode=RCTM_User, time=0.025, value=1}
        <at 0.125 0.2857>
        <at 0.9 0.3>
        {arriveTangent=-0.0285, interpMode=RCIM_Cubic, leaveTangent=-0.0285, tangentMode=RCTM_User, time=1, value=0}
       ]}}}
   )
  ]
  rendererClasses [&NiagaraSpriteRendererProperties]
  renderers [{rendererClass=&NiagaraSpriteRendererProperties}]
 }
 Streaks {
  emitterUpdateScript [
   &EmitterState(Enable Distance Culling=true, MaxDistance=8000, Module.bUseMaxDistance=true, Module.bUseMinDistance=false, Scalability Mode=::NewEnumerator1)
   &SpawnBurst_Instantaneous(Spawn Count=15, Use Spawn Probability=false)
  ]
  particleSpawnScript [
   /Niagara/Modules/Spawn/Initialization/InitializeParticle(
    Color=<rgba 0.1198 0.1072 0.0898 1>
    Lifetime=&UniformRangedFloat(Minimum=0.25, Maximum=0.35)
    Module.Write Color=true
    Module.Write Lifetime=true
    Module.Write Mass=true
    Module.Write Material Random=true
    Module.Write Position=false
    Module.Write Scale=false
    Module.Write SpriteRotation=false
    Module.Write SpriteSize=true
    Position=&SimulationPosition()
    Sprite Size=&UniformRangedVector2D(Minimum=<20 100>, Maximum=<50 100>)
   ) {enabled=false}
   /Niagara/Modules/Spawn/Initialization/V2/InitializeParticle(
    Color=<rgba 0.1198 0.1072 0.0898 1>
    Lifetime=&UniformRangedFloat(Minimum=0.2, Maximum=0.3)
    Position Offset=<0 0 -5>
    Sprite Size Max=<100 300>
    Sprite Size Min=<20 100>
    Sprite Size Mode=::NewEnumerator2
    UsePositionOffset=true
    Write Material Random=true
   ) {module=InitializeParticle001}
   &BoxLocation(
    Box Size=<600 600 0>
    MaskSpawnGroup=false
    Module.Override Seed=false
    Rotate Box=false
    Surface Expansion Mode:ENiagaraExpansionMode=::NewEnumerator2
    Surface Only=true
    Surface Only Band Thickness=2
   ) {enabled=false}
   &AddVelocityFromPoint(
    Module.Use Falloff Distance=false
    Origin Offset=&UniformRangedVector(Minimum=<0 0 0>, Maximum=<0 0 0>)
    Velocity Strength=&UniformRangedFloat(Minimum=15, Maximum=25)
   )
   /Niagara/Modules/Update/SubUV/SubUVAnimation(SubUV Animation Mode=::NewEnumerator1)
   &ShapeLocation(Handle Radius=1, Large Radius=2, MaskSpawnGroup=false, Override Seed=false, Shape Primitive=::NewEnumerator3, UseOwnerScale=true)
   &AddVelocity(
    Module.Use Falloff Distance=false
    Origin Offset=<0 0 -7>
    Override Seed=false
    Velocity Mode=::NewEnumerator1
    Velocity Speed=&RandomRangeFloat(Minimum=150, Maximum=250)
   ) {module=AddVelocity001}
   &CameraOffset(Camera Offset Amount=3)
  ]
  particleUpdateScript [
   &ParticleState()
   &Drag(Use Linear Drag=true, Use Rotational Drag=false)
   &SolveForcesAndVelocity(Clamp Velocity=false, Limit Acceleration=false)
   &ScaleColor(
    Scale Alpha=&FloatFromCurve(FloatCurve=DataInterface() {propertyValues {Curve {keys [
         {interpMode=RCIM_Cubic, tangentMode=RCTM_User, time=0.0003, value=0.9198}
         {arriveTangent=-0.3498, interpMode=RCIM_Cubic, leaveTangent=-0.3498, time=0.1, value=1.3}
         {arriveTangent=-2.7704, interpMode=RCIM_Cubic, leaveTangent=-0.8494, tangentMode=RCTM_Break, time=0.2, value=0.85}
         {arriveTangent=-0.9201, interpMode=RCIM_Cubic, leaveTangent=-0.9201, tangentMode=RCTM_User, time=0.5792, value=0.6723}
         <at 1 0>
        ]}}})
    Scale RGB=&Multiply_Vector(
     A=&VectorFromFloat(Value=&FloatFromCurve(FloatCurve=DataInterface() {propertyValues {Curve {keys [<at 0 1>,<at 0.1 2>,<at 0.2 1>,<at 1 1 RCIM_Cubic>]}}}))
     B=&MakeVectorFromLinearColorRGB(Color=&FNTimeOfDayColor(
       Night Color=<rgba 0.1771 0.2756 1 1>
       Morning Color=<rgba 0.5607 0.6584 1 1>
       Day Color=<rgba 1 1 1 1>
       Evening Color=<rgba 1 0.8696 0.5989 1>
      ))
    )
    ScaleA=true
    ScaleRGB=true
   )
   &ScaleSpriteSize(Scale Factor=&Vector2DFromCurve(
     Vector2Curve=DataInterface() {propertyValues {
       XCurve {keys [{tangentMode=RCTM_User, time=0, value=0.575}, {tangentMode=RCTM_User, time=1, value=0}]}
       YCurve {keys [{tangentMode=RCTM_User, time=0, value=0.25}, {tangentMode=RCTM_User, time=1, value=1}]}
      }}
     ScaleCurve=<0.7 0.3>
    ))
  ]
  rendererClasses [&NiagaraSpriteRendererProperties]
  renderers [{rendererClass=&NiagaraSpriteRendererProperties}]
 }
}
summary {
 systemName=NS_Emote_CrumbViolin_FootSlam
 emitters [
  {emitterName=SmallBits, simTarget=GPUComputeSim, rendererClasses [&NiagaraSpriteRendererProperties]}
  {emitterName=GroundDust, rendererClasses [&NiagaraSpriteRendererProperties]}
  {emitterName=Streaks, rendererClasses [&NiagaraSpriteRendererProperties]}
  {emitterName=MainBurst, rendererClasses [&NiagaraRibbonRendererProperties, &NiagaraSpriteRendererProperties]}
 ]
 userVariables [
  User.DesertBoulder {type=NiagaraBool}
  User.Reveal {type=NiagaraFloat, defaultValue=0.5, description=""}
 ]
}
