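# Niagara steam effect for the CrumbViolin emote, built from bursts that feed one ribbon and two sprite renderers.
# User.Reveal and system age drive plume length, width and opacity, with per-system and per-emitter random variation.
# Ribbon points are placed by normalized exec index and sprites cycle along the plume, while spawn group 1 adds short-lived puffs with cone velocity.
# Time-of-day inputs tint the steam; the matching MainBurst001 emitter is disabled.
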NiagaraSystem /BRCosmetics/Animation/Game/MainPlayer/Emotes/CrumbViolin/FX/NS_Emote_CrumbViolin_Steam
dependencies {
 usedDataInterfaces [
  &NiagaraDataInterfaceCurve
  &NiagaraDataInterfaceColorCurve
  &NiagaraDataInterfaceSpriteRendererInfo
 ]
 usedDynamicInputs [&FNTimeOfDayScalar, &ClampFloat]
 usedModules [
  &self:SystemSpawnScript.SystemScriptSource.SystemScriptGraph.NiagaraNodeAssignment_0.SetVariables_#1
  &self:SystemSpawnScript.SystemScriptSource.SystemScriptGraph.NiagaraNodeAssignment_3.SetVariables_#2
  &SystemState
  &self:SystemSpawnScript.SystemScriptSource.SystemScriptGraph.NiagaraNodeAssignment_1.SetVariables_#3
  &InterpolateOverTime
  &self:SystemSpawnScript.SystemScriptSource.SystemScriptGraph.NiagaraNodeAssignment_2.SetVariables_#4
 ]
}
emitters {
 MainBurst {
  emitterSpawnScript [
   /SetVariables(
    Emitter.Length=85
    Emitter.ParticlesToSpawn=12
    Emitter.Random:NiagaraFloat=&RandomRangeFloat(Minimum=0, Maximum=1)
   ) {isSetParameters=true, module=SetVariables_#5}
   /SetVariables(
    Emitter.SpriteLength=70
    Emitter.SpriteRandom:NiagaraFloat=&RandomRangeFloat(Minimum=0, Maximum=1)
    Emitter.SpritesToSpawn=3
   ) {isSetParameters=true, module=SetVariables_#6}
  ]
  emitterUpdateScript [
   &EmitterState(Life Cycle Mode=::NewEnumerator1, Loop Behavior=::NewEnumerator1, Loop Delay=0.2, Loop Duration=1, UseLoopDelay=false)
   &SpawnBurst_Instantaneous(Spawn Count=$Emitter.ParticlesToSpawn, Spawn Time=0.2, Use Spawn Probability=false, UseLoopCountLimit=false)
   &SpawnBurst_Instantaneous(Spawn Count=$Emitter.SpritesToSpawn, Spawn Probability=0.5, Use Spawn Probability=false, UseLoopCountLimit=false) {module=SpawnBurst_Instantaneous001}
   &SpawnBurst_Instantaneous(
    Spawn Count=&RandomRangeInt(Minimum=8, Maximum=12)
    Spawn Group=1
    Spawn Time=0.03
    Use Spawn Probability=false
    UseLoopCountLimit=false
   ) {module=SpawnBurst_Instantaneous002}
   /SetVariables(Emitter.Reveal:NiagaraFloat=&Multiply_Float(
     A=&Lerp_Float(
      A=&FloatFromCurve(FloatCurve=DataInterface() {propertyValues {Curve {keys [<at 0 0>,<at 0.6 1>]}}}, CurveIndex=$User.Reveal)
      B=&FloatFromCurve(FloatCurve=DataInterface() {propertyValues {Curve {keys [<at 0 0>,<at 0.95 1>]}}}, CurveIndex=$User.Reveal)
      Alpha=$System.SystemRandom
     )
     B=&Lerp_Float(A=1, B=1.1, Alpha=$Emitter.Random)
    )) {isSetParameters=true, module=SetVariables_#7}
   /SetVariables(Emitter.Length:NiagaraFloat=&Multiply_Float(
     A=&FloatFromCurve(
      FloatCurve=DataInterface() {propertyValues {Curve {keys [
          {arriveTangent=2.5813, interpMode=RCIM_Cubic, leaveTangent=2.5813, tangentMode=RCTM_User, time=0, value=0.1}
          {tangentMode=RCTM_User, time=0.5, value=1}
         ]}}}
      CurveIndex=$Engine.System.Age
     )
     B=&Multiply_Float(
      A=$Emitter.Initial.Length
      B=&FloatFromCurve(
       FloatCurve=DataInterface() {propertyValues {Curve {keys [
           {interpMode=RCIM_Cubic, leaveTangent=6.4155, tangentMode=RCTM_Break, time=0, value=0}
           {arriveTangent=0.1191, interpMode=RCIM_Cubic, leaveTangent=0.1191, tangentMode=RCTM_User, time=0.1, value=0.9}
           {arriveTangent=0.1177, interpMode=RCIM_Cubic, leaveTangent=0.1177, tangentMode=RCTM_User, time=1, value=1}
          ]}}}
       CurveIndex=$Emitter.Reveal
       Scale Curve=&Lerp_Float(A=0.7, B=1, Alpha=$System.SizeRandom)
      )
     )
    )) {isSetParameters=true, module=SetVariables_#8}
   /SetVariables(Emitter.SpriteReveal:NiagaraFloat=&Multiply_Float(
     A=&Lerp_Float(
      A=&FloatFromCurve(FloatCurve=DataInterface() {propertyValues {Curve {keys [<at 0 0>,<at 0.6 1>]}}}, CurveIndex=$User.Reveal)
      B=&FloatFromCurve(FloatCurve=DataInterface() {propertyValues {Curve {keys [<at 0 0>,<at 0.95 1>]}}}, CurveIndex=$User.Reveal)
      Alpha=$System.SystemRandom
     )
     B=&Lerp_Float(A=1, B=1.1, Alpha=$Emitter.Random)
    )) {isSetParameters=true, module=SetVariables_#9}
   /SetVariables(Emitter.SpriteLength:NiagaraFloat=&Multiply_Float(
     A=&FloatFromCurve(
      FloatCurve=DataInterface() {propertyValues {Curve {keys [
          {arriveTangent=2.5813, interpMode=RCIM_Cubic, leaveTangent=2.5813, tangentMode=RCTM_User, time=0, value=0.1}
          {tangentMode=RCTM_User, time=0.5, value=1}
         ]}}}
      CurveIndex=$Engine.System.Age
     )
     B=&Multiply_Float(
      A=$Emitter.Initial.SpriteLength
      B=&FloatFromCurve(
       FloatCurve=DataInterface() {propertyValues {Curve {keys [
           {interpMode=RCIM_Cubic, leaveTangent=6.4155, tangentMode=RCTM_Break, time=0, value=0}
           {arriveTangent=0.1191, interpMode=RCIM_Cubic, leaveTangent=0.1191, tangentMode=RCTM_User, time=0.1, value=0.9}
           {arriveTangent=0.1177, interpMode=RCIM_Cubic, leaveTangent=0.1177, tangentMode=RCTM_User, time=1, value=1}
          ]}}}
       CurveIndex=$Emitter.SpriteReveal
       Scale Curve=&Lerp_Float(A=0.7, B=1, Alpha=$System.SizeRandom)
      )
     )
    )) {isSetParameters=true, module=SetVariables_#10}
  ]
  particleSpawnScript [
   &InitializeParticle(
    Color=&MaskLinearColorBySpawnGroup(Masked Linear Color=<rgba 1 1 1 1>, Passthrough Linear Color=<rgba 1 1 1 0.6>)
    Lifetime=&MaskFloatBySpawnGroup(Masked Float=$System.CurrentLoopDuration, Passthrough Float=&UniformRangedFloat(Minimum=0.4, Maximum=0.6))
    Position Offset=&MaskVectorBySpawnGroup(Masked Vector=<0 0 0>, Passthrough Vector=<0 -5 0>)
    Sprite Rotation Mode=::NewEnumerator2
    Sprite Size Mode=::NewEnumerator3
    Sprite UV Mode=::NewEnumerator4
    Uniform Sprite Size=&MaskFloatBySpawnGroup(Masked Float=10, Passthrough Float=&UniformRangedFloat(Minimum=10, Maximum=15))
    UsePositionOffset=true
    Write Material Random=true
   )
   &DynamicMaterialParameters(
    Index 0 Param 1=$Emitter.Random
    Module.Index0Param0Write=true
    Module.Index0Param1Write=true
    Module.Index0Param2Write=true
    Module.Index0Param3Write=true
   )
   /SetVariables(Particles.VisibilityTag=&SpawnGroup()) {isSetParameters=true, module=SetVariables_#11}
   &ShapeLocation(
    MaskSpawnGroup=false
    Offset Coordinate Space=::Local
    Offset Mode=::NewEnumerator0
    Override Seed=false
    Sphere Radius=&MaskFloatBySpawnGroup(Masked Float=0, Passthrough Float=0.1)
    UseOwnerScale=true
   ) {module=ShapeLocation001}
   &ApplyInitialForces()
   &AddVelocity(
    Cone Angle=5
    Cone Axis=<1 0.9 0>
    Module.UseConeFalloff=false
    Module.UseVelDistribution=false
    Override Seed=false
    Velocity Mode=::NewEnumerator2
    Velocity Speed=&MaskFloatBySpawnGroup(
     Masked Float=0
     Passthrough Float=&Multiply_Float(
      A=&RandomRangeFloat(Minimum=2, Maximum=250)
      B=&Lerp_Float(A=0.7, B=1.2, Alpha=$System.SizeRandom)
     )
    )
   )
  ]
  particleUpdateScript [
   &SolveForcesAndVelocity(Clamp Velocity=false, Limit Acceleration=false)
   &ParticleState()
   /SetVariables(Particles.ParticleRelativeLifetimeRibbons:NiagaraFloat=&ReturnNormalizedExecIndex()) {isSetParameters=true, module=SetVariables_#12}
   /SetVariables(Particles.OpacityMultiplierRibbons:NiagaraFloat=&FloatFromCurve(
     FloatCurve=DataInterface() {propertyValues {Curve {keys [
         {interpMode=RCIM_Cubic, tangentMode=RCTM_User, time=0, value=0}
         {arriveTangent=0.007, interpMode=RCIM_Cubic, leaveTangent=0.007, tangentMode=RCTM_User, time=0.5, value=1.5}
         {arriveTangent=-0.7513, interpMode=RCIM_Cubic, leaveTangent=-0.7513, tangentMode=RCTM_User, time=0.9, value=1}
         {arriveTangent=-0.9808, interpMode=RCIM_Cubic, leaveTangent=-0.9808, tangentMode=RCTM_User, time=1.8, value=0}
        ]}}}
     CurveIndex=&Add_Float(
      A=&Multiply_Float(A=$Particles.ParticleRelativeLifetimeRibbons, B=-1)
      B=&Multiply_Float(A=$Emitter.Reveal, B=2)
     )
    )) {isSetParameters=true, module=SetVariables_#13}
   /SetVariables(Particles.InterpolateCurveRibbons:NiagaraFloat=&FloatFromCurve(
     FloatCurve=DataInterface() {propertyValues {Curve {keys [{arriveTangent=-0.6503, time=0, value=0}, {arriveTangent=0.8679, time=1, value=1}]}}}
     CurveIndex=$Particles.ParticleRelativeLifetimeRibbons
    )) {isSetParameters=true, module=SetVariables_#14}
   /SetVariables(Particles.PositionRibbons:NiagaraPosition=&AddVectorToPosition(
     Position=&AddVectorToPosition(
      Position=/Niagara/DynamicInputs/Helpers/SimulationPosition()
      Vector=&TransformVector(Vector=<0 -5 0>, Destination Space=::Simulation)
     )
     Vector=&VectorFromFloat(Value=&Abs_Float(Float:NiagaraFloat=&NormalizeFloat(Value=&Multiply_Float(A=$Emitter.Length, B=$Particles.InterpolateCurveRibbons), Min=-1, Max=1)))
    )) {isSetParameters=true, module=SetVariables_#15}
   /SetVariables(
    Particles.RibbonColor:LinearColor=&MultiplyLinearColor(
     Linear Color=&MultiplyLinearColor(
      Linear Color=&ColorFromCurve(
       ColorCurve=DataInterface() {propertyValues {
         AlphaCurve {keys [
           <at 0 0>
           <at 0.0346 1>
           <at 0.2133 0.8>
           {arriveTangent=-1.0175, interpMode=RCIM_Cubic, leaveTangent=-1.0175, time=0.7268, value=0.8}
           <at 0.9995 0 RCIM_Cubic>
          ]}
         BlueCurve {keys [<at 0.4438 1>]}
         GreenCurve {keys [<at 0.4438 1>]}
         RedCurve {keys [<at 0.4438 1>]}
        }}
       CurveIndex=$Particles.ParticleRelativeLifetimeRibbons
      )
      Scale Factor=&MakeVector4_FromLinearColor(LinearColor=&FNTimeOfDayColor(Night Color=<rgba 0.1771 0.2756 1 0.7>, Morning Color=<rgba 1 1 1 1>, Day Color=<rgba 1 1 1 1>, Evening Color=<rgba 1 0.8696 0.5989 1>))
     )
     Scale Factor=&MakeFloat4FromVectorAndFloat(
      "Vector (RGB)"=&VectorFromFloat(Value=&FloatFromCurve(FloatCurve=DataInterface() {propertyValues {Curve {keys [<at 0.45 1 RCIM_Cubic>]}}}, CurveIndex=$User.Reveal))
      "Float (Alpha)"=1
     )
    )
    Particles.RibbonWidth=&Multiply_Float(
     A=&FloatFromCurve(
      FloatCurve=DataInterface() {propertyValues {Curve {keys [
          {arriveTangent=0.675, interpMode=RCIM_Cubic, leaveTangent=0.675, tangentMode=RCTM_User, time=0, value=0.2}
          {arriveTangent=1.0523, interpMode=RCIM_Cubic, leaveTangent=1.0523, tangentMode=RCTM_User, time=0.2799, value=0.5408}
          {arriveTangent=-0.0372, interpMode=RCIM_Cubic, leaveTangent=-0.0372, tangentMode=RCTM_User, time=0.7, value=0.8}
          {arriveTangent=-1.5957, interpMode=RCIM_Cubic, leaveTangent=-1.5957, tangentMode=RCTM_User, time=1, value=0.5138}
         ]}}}
      CurveIndex=$Particles.InterpolateCurveRibbons
      Scale Curve=15
     )
     B=&FloatFromCurve(
      FloatCurve=DataInterface() {propertyValues {Curve {keys [
          {arriveTangent=5.9942, interpMode=RCIM_Cubic, leaveTangent=5.9942, tangentMode=RCTM_User, time=0, value=0.475}
          {arriveTangent=0.0898, interpMode=RCIM_Cubic, leaveTangent=0.0898, tangentMode=RCTM_User, time=0.3, value=1}
         ]}}}
      CurveIndex=$System.Age
     )
    )
   ) {isSetParameters=true, module=SetVariables_#16}
   /SetVariables(Particles.ScaledRibbonWidth:NiagaraFloat=&Multiply_Float(
     A=$Particles.RibbonWidth
     B=&Multiply_Float(
      A=&FloatFromCurve(
       FloatCurve=DataInterface() {propertyValues {Curve {keys [
           {interpMode=RCIM_Cubic, leaveTangent=0.5491, tangentMode=RCTM_Break, time=0, value=0.7}
           {arriveTangent=1.0086, interpMode=RCIM_Cubic, leaveTangent=1.0086, tangentMode=RCTM_User, time=0.3931, value=1}
           {arriveTangent=0.0216, interpMode=RCIM_Cubic, tangentMode=RCTM_Break, time=1, value=1.5}
          ]}}}
       CurveIndex=$Emitter.Reveal
       Scale Curve=&Lerp_Float(
        A=1.6
        B=&FloatFromCurve(
         FloatCurve=DataInterface() {propertyValues {Curve {keys [<at 0 1.9 RCIM_Constant>,<at 0.333 2.25 RCIM_Constant>,<at 0.666 2.6 RCIM_Constant>]}}}
         CurveIndex=$Emitter.Random
        )
        Alpha=$System.SizeRandom
       )
      )
      B=1
     )
    )) {isSetParameters=true, module=SetVariables_#17}
   /SetVariables(Particles.ScaledRibbonColor:LinearColor=&Multiply_LinearColorByFloat(
     Linear Color=&MultiplyLinearColor(
      Linear Color=$Particles.RibbonColor
      Scale Factor=&MakeFloat4FromVectorAndFloat(
       "Vector (RGB)"=<1 1 1>
       "Float (Alpha)"=&Multiply_Float(
        A=$Particles.OpacityMultiplierRibbons
        B=&Multiply_Float(
         A=&FloatFromCurve(
          FloatCurve=DataInterface() {propertyValues {Curve {keys [
              {interpMode=RCIM_Cubic, tangentMode=RCTM_Break, time=0.6, value=1}
              {arriveTangent=-4.2498, interpMode=RCIM_Cubic, tangentMode=RCTM_Break, time=0.99, value=0}
             ]}}}
          CurveIndex=$Emitter.Reveal
          Scale Curve=&Lerp_Float(A=0.8, B=1, Alpha=$Emitter.Random)
         )
         B=$System.OpacityRandom
        )
       )
      )
     )
     Float=&FloatFromCurve(
      FloatCurve=DataInterface() {propertyValues {Curve {keys [
          {arriveTangent=5.7153, interpMode=RCIM_Cubic, leaveTangent=5.7153, tangentMode=RCTM_User, time=0, value=0}
          {arriveTangent=-1.4405, interpMode=RCIM_Cubic, leaveTangent=-1.4405, tangentMode=RCTM_User, time=0.3, value=2}
          {arriveTangent=-1.4711, interpMode=RCIM_Cubic, leaveTangent=-1.4711, tangentMode=RCTM_User, time=1, value=0}
         ]}}}
      CurveIndex=$System.Age
     )
    )) {isSetParameters=true, module=SetVariables_#18}
   &CameraOffset(Camera Offset Amount=5) {enabled=false}
   &ParticleState() {module=ParticleState001}
   &ScaleSpriteSize(Scale Factor=&MaskVector2dBySpawnGroup(
     Masked Vector2D=$Particles.Initial.SpriteSize
     Passthrough Vector2D=&Multiply_Vector2D_ByFloat(
      Float=&FloatFromCurve(
       FloatCurve=DataInterface() {propertyValues {Curve {keys [
           {arriveTangent=5.7302, interpMode=RCIM_Cubic, leaveTangent=5.7302, tangentMode=RCTM_User, time=0, value=0}
           {arriveTangent=0.2869, interpMode=RCIM_Cubic, leaveTangent=0.2869, tangentMode=RCTM_User, time=0.5, value=0.975}
          ]}}}
       CurveIndex=$Engine.System.Age
      )
      Vector2D=&Vector2DFromFloat(Value=&FloatFromCurve(
        FloatCurve=DataInterface() {propertyValues {Curve {keys [{tangentMode=RCTM_User, time=0, value=0.4}, <at 0.1 0.1>, {tangentMode=RCTM_User, time=1, value=1}]}}}
        Scale Curve=&Multiply_Float(A=0.7, B=&Lerp_Float(A=0.7, B=1, Alpha=$System.SizeRandom))
       ))
     )
     Spawn Group Mask=0
    )) {module=ScaleSpriteSize001}
   &SpriteRotationRate(Rotation Rate=&Multiply_Float(
     A=&RandomRangeFloat(Minimum=100, Maximum=150)
     B=&UniformAOrBFloat(A=-1, Evaluation Type=::NewEnumerator0, Recalculate Random Each Loop=true, Override Seed=false)
    ))
   /SetVariables(Particles.ParticleRelativeLifetimeSprites:NiagaraFloat=&Frac_Float(Float=&Add_Float(A=&Multiply_Float(A=1.5, B=$System.Age), B=&ReturnNormalizedExecIndex()))) {isSetParameters=true, module=SetVariables_#19}
   /SetVariables(Particles.OpacityMultiplierSprites:NiagaraFloat=&FloatFromCurve(
     FloatCurve=DataInterface() {propertyValues {Curve {keys [
         {interpMode=RCIM_Cubic, tangentMode=RCTM_User, time=0, value=0}
         {arriveTangent=-0.6677, interpMode=RCIM_Cubic, leaveTangent=-0.6677, tangentMode=RCTM_User, time=0.2, value=1.5}
         {arriveTangent=-0.6675, tangentMode=RCTM_User, time=1.5, value=0}
        ]}}}
     CurveIndex=&Add_Float(
      A=&Multiply_Float(A=$Particles.ParticleRelativeLifetimeSprites, B=-1)
      B=&Multiply_Float(A=$Emitter.SpriteReveal, B=&RandomRangeFloat(Minimum=1, Maximum=2))
     )
    )) {isSetParameters=true, module=SetVariables_#20}
   /SetVariables(Particles.InterpolateCurveSprites:NiagaraFloat=&FloatFromCurve(
     FloatCurve=DataInterface() {propertyValues {Curve {keys [
         {arriveTangent=-0.6503, interpMode=RCIM_Cubic, leaveTangent=1.4736, tangentMode=RCTM_Break, time=0, value=0}
         {arriveTangent=0.4202, interpMode=RCIM_Cubic, tangentMode=RCTM_Break, time=1, value=1}
        ]}}}
     CurveIndex=$Particles.ParticleRelativeLifetimeSprites
    )) {isSetParameters=true, module=SetVariables_#21}
   /SetVariables(Particles.PositionSprites:NiagaraPosition=&AddVectorToPosition(
     Position=&AddVectorToPosition(
      Position=&SimulationPosition()
      Vector=&TransformVector(Vector=<0 -5 0>, Destination Space=::Simulation)
     )
     Vector=&VectorFromFloat(Value=&Abs_Float(Float:NiagaraFloat=&NormalizeFloat(Value=&Multiply_Float(A=$Emitter.SpriteLength, B=$Particles.InterpolateCurveSprites), Min=-1, Max=1)))
    )) {isSetParameters=true, module=SetVariables_#22}
   /SetVariables(Particles.SpriteSize=&MaskVector2dBySpawnGroup(
     Masked Vector2D=&Multiply_Vector2D_ByFloat(
      Float=&FloatFromCurve(
       FloatCurve=DataInterface() {propertyValues {Curve {keys [
           {arriveTangent=4.522, interpMode=RCIM_Cubic, leaveTangent=4.522, tangentMode=RCTM_User, time=0, value=0.475}
           {arriveTangent=0.0898, interpMode=RCIM_Cubic, leaveTangent=0.0898, tangentMode=RCTM_User, time=0.3, value=1}
          ]}}}
       CurveIndex=$System.Age
      )
      Vector2D=&Multiply_Vector2D(
       A=&Vector2DFromFloat(Value=&FloatFromCurve(
         FloatCurve=DataInterface() {propertyValues {Curve {keys [{tangentMode=RCTM_User, time=0, value=0}, {tangentMode=RCTM_User, time=1, value=1}]}}}
         CurveIndex=$Particles.InterpolateCurveSprites
         Scale Curve=&Lerp_Float(A=35, B=50, Alpha=&RandomRangeFloat(Minimum=0, Maximum=1))
        ))
       B=&Vector2DFromFloat(Value=&Lerp_Float(A=0.7, B=1, Alpha=$System.SizeRandom))
      )
     )
     Passthrough Vector2D=$Particles.Initial.SpriteSize
     Spawn Group Mask=0
    )) {isSetParameters=true, module=SetVariables_#23}
   &SubUVAnimation(Start Frame=20, SubUV Lookup Index:NiagaraFloat=$Particles.ParticleRelativeLifetimeSprites)
   &SubUVAnimation(
    Output Alternate Renderer Binding=true
    "Sprite Renderer "=DataInterface() {propertyValues {SpriteRenderer=&self:MainBurst_1.NiagaraSpriteRendererProperties_1}}
    Start Frame Range Override=40
    UseEndFrame=true
    UseStartFrame=true
   ) {module=SubUVAnimation002}
   &ScaleColor(
    Scale Alpha=&MaskFloatBySpawnGroup(
     Masked Float=&Multiply_Float(
      A=&Multiply_Float(
       A=&FloatFromCurve(
        FloatCurve=DataInterface() {propertyValues {Curve {keys [
            {arriveTangent=16.5612, interpMode=RCIM_Cubic, leaveTangentWeight=0.3333, time=0, value=0}
            {arriveTangentWeight=0.1333, interpMode=RCIM_Cubic, leaveTangentWeight=0.1667, tangentMode=RCTM_User, time=0.333, value=1}
            {arriveTangentWeight=0.1667, interpMode=RCIM_Cubic, tangentMode=RCTM_User, time=1, value=0}
           ]}}}
        CurveIndex=$Particles.ParticleRelativeLifetimeSprites
        Scale Curve=&RandomRangeFloat(Minimum=0.5, Maximum=1)
       )
       B=&Multiply_Float(
        A=&FloatFromCurve(
         FloatCurve=DataInterface() {propertyValues {Curve {keys [
             {interpMode=RCIM_Cubic, tangentMode=RCTM_Break, time=0.6, value=1}
             {arriveTangent=-4.2498, interpMode=RCIM_Cubic, tangentMode=RCTM_Break, time=0.99, value=0}
            ]}}}
         CurveIndex=$Emitter.SpriteReveal
         Scale Curve=&Lerp_Float(A=1, B=2, Alpha=$Emitter.Random)
        )
        B=$System.OpacityRandom
       )
      )
      B=$Particles.OpacityMultiplierSprites
     )
     Passthrough Float=&Multiply_Float(
      A=&FloatFromCurve(
       FloatCurve=DataInterface() {propertyValues {Curve {keys [
           {arriveTangent=5.5857, interpMode=RCIM_Cubic, leaveTangent=5.5857, leaveTangentWeight=0.3333, tangentMode=RCTM_User, time=0, value=0.3}
           {arriveTangentWeight=0.1333, interpMode=RCIM_Cubic, leaveTangentWeight=0.1667, tangentMode=RCTM_User, time=0.1, value=2}
           {arriveTangentWeight=0.1667, interpMode=RCIM_Cubic, tangentMode=RCTM_User, time=1, value=0}
          ]}}}
       Scale Curve=&Multiply_Float(A=1, B=$System.OpacityRandom)
      )
      B=&RandomRangeFloat(Minimum=0.5, Maximum=1)
     )
    )
    Scale RGB=&MakeVectorFromLinearColorRGB(Color=&FNTimeOfDayColor(
      Night Color=<rgba 0.1771 0.2756 1 0.7>
      Morning Color=<rgba 0.5607 0.6584 1 1>
      Day Color=<rgba 1 1 1 1>
      Evening Color=<rgba 1 0.8696 0.5989 1>
     ))
    ScaleA=true
    ScaleRGB=true
   )
   &ScaleColor(
    Scale Alpha=&MaskFloatBySpawnGroup(
     Masked Float=&Multiply_Float(
      A=$Particles.OpacityMultiplierSprites
      B=&FloatFromCurve(
       FloatCurve=DataInterface() {propertyValues {Curve {keys [
           {arriveTangent=9.0725, interpMode=RCIM_Cubic, leaveTangent=9.0725, tangentMode=RCTM_User, time=0, value=0}
           {arriveTangent=-0.6131, interpMode=RCIM_Cubic, leaveTangent=-0.6131, tangentMode=RCTM_User, time=0.1408, value=0.9847}
           {arriveTangent=-1.4711, interpMode=RCIM_Cubic, leaveTangent=-1.4711, tangentMode=RCTM_User, time=1, value=0}
          ]}}}
       CurveIndex=$Emitter.Age
      )
     )
     Passthrough Float=&FloatFromCurve(
      FloatCurve=DataInterface() {propertyValues {Curve {keys [
          {arriveTangent=9.0725, interpMode=RCIM_Cubic, leaveTangent=9.0725, tangentMode=RCTM_User, time=0, value=0}
          {arriveTangent=-2.5249, interpMode=RCIM_Cubic, leaveTangent=-2.5249, tangentMode=RCTM_User, time=0.1408, value=0.9847}
          {arriveTangent=-0.0405, interpMode=RCIM_Cubic, leaveTangent=-0.0405, tangentMode=RCTM_User, time=1, value=0}
         ]}}}
      CurveIndex=$Emitter.Age
     )
    )
    ScaleA=true
    ScaleRGB=false
   ) {module=ScaleColor001}
   &ScaleColor(
    Scale Alpha=&MaskFloatBySpawnGroup(
     Masked Float=&Multiply_Float(
      A=$Particles.OpacityMultiplierSprites
      B=&FloatFromCurve(
       FloatCurve=DataInterface() {propertyValues {Curve {keys [
           {arriveTangent=9.0725, interpMode=RCIM_Cubic, leaveTangent=9.0725, tangentMode=RCTM_User, time=0, value=0}
           {arriveTangent=-0.6131, interpMode=RCIM_Cubic, leaveTangent=-0.6131, tangentMode=RCTM_User, time=0.1408, value=0.9847}
           {arriveTangent=-1.4711, interpMode=RCIM_Cubic, leaveTangent=-1.4711, tangentMode=RCTM_User, time=1, value=0}
          ]}}}
       CurveIndex=$Emitter.Age
      )
     )
     Passthrough Float=&FloatFromCurve(
      FloatCurve=DataInterface() {propertyValues {Curve {keys [
          {arriveTangent=9.0725, interpMode=RCIM_Cubic, leaveTangent=9.0725, tangentMode=RCTM_User, time=0, value=0}
          {arriveTangent=-2.5249, interpMode=RCIM_Cubic, leaveTangent=-2.5249, tangentMode=RCTM_User, time=0.1408, value=0.9847}
          {arriveTangent=-0.0405, interpMode=RCIM_Cubic, leaveTangent=-0.0405, tangentMode=RCTM_User, time=1, value=0}
         ]}}}
      Scale Curve=&Multiply_Float(A=1, B=$System.ToD)
     )
    )
    ScaleA=true
    ScaleRGB=false
   ) {module=ScaleColor002}
  ]
  rendererClasses [&NiagaraRibbonRendererProperties, &NiagaraSpriteRendererProperties]
  renderers [
   {rendererClass=&NiagaraRibbonRendererProperties}
   {rendererClass=&NiagaraSpriteRendererProperties, rendererIndex=1}
   {rendererClass=&NiagaraSpriteRendererProperties, rendererIndex=2}
  ]
 }
 MainBurst001 {
  bEnabled=false
  emitterSpawnScript [
   /SetVariables(
    Emitter.Length=85
    Emitter.ParticlesToSpawn=12
    Emitter.Random:NiagaraFloat=&RandomRangeFloat(Minimum=0, Maximum=1)
   ) {isSetParameters=true, module=SetVariables_#5}
   /SetVariables(
    Emitter.SpriteLength=70
    Emitter.SpriteRandom:NiagaraFloat=&RandomRangeFloat(Minimum=0, Maximum=1)
    Emitter.SpritesToSpawn=3
   ) {isSetParameters=true, module=SetVariables_#6}
  ]
  emitterUpdateScript [
   &EmitterState(Life Cycle Mode=::NewEnumerator1, Loop Behavior=::NewEnumerator1, Loop Delay=0.2, Loop Duration=1, UseLoopDelay=false)
   &SpawnBurst_Instantaneous(Spawn Count=$Emitter.ParticlesToSpawn, Spawn Time=0.2, Use Spawn Probability=false, UseLoopCountLimit=false)
   &SpawnBurst_Instantaneous(Spawn Count=$Emitter.SpritesToSpawn, Spawn Probability=0.5, Use Spawn Probability=false, UseLoopCountLimit=false) {module=SpawnBurst_Instantaneous001}
   &SpawnBurst_Instantaneous(
    Spawn Count=&RandomRangeInt(Minimum=8, Maximum=12)
    Spawn Group=1
    Spawn Time=0.03
    Use Spawn Probability=false
    UseLoopCountLimit=false
   ) {module=SpawnBurst_Instantaneous002}
   /SetVariables(Emitter.Reveal:NiagaraFloat=&Multiply_Float(
     A=&Lerp_Float(
      A=&FloatFromCurve(FloatCurve=DataInterface() {propertyValues {Curve {keys [<at 0 0>,<at 0.6 1>]}}}, CurveIndex=$User.Reveal)
      B=&FloatFromCurve(FloatCurve=DataInterface() {propertyValues {Curve {keys [<at 0 0>,<at 0.95 1>]}}}, CurveIndex=$User.Reveal)
      Alpha=$System.SystemRandom
     )
     B=&Lerp_Float(A=1, B=1.1, Alpha=$Emitter.Random)
    )) {isSetParameters=true, module=SetVariables_#7}
   /SetVariables(Emitter.Length:NiagaraFloat=&Multiply_Float(
     A=&FloatFromCurve(
      FloatCurve=DataInterface() {propertyValues {Curve {keys [
          {arriveTangent=2.5813, interpMode=RCIM_Cubic, leaveTangent=2.5813, tangentMode=RCTM_User, time=0, value=0.1}
          {tangentMode=RCTM_User, time=0.5, value=1}
         ]}}}
      CurveIndex=$Engine.System.Age
     )
     B=&Multiply_Float(
      A=$Emitter.Initial.Length
      B=&FloatFromCurve(
       FloatCurve=DataInterface() {propertyValues {Curve {keys [
           {interpMode=RCIM_Cubic, leaveTangent=6.4155, tangentMode=RCTM_Break, time=0, value=0}
           {arriveTangent=0.1191, interpMode=RCIM_Cubic, leaveTangent=0.1191, tangentMode=RCTM_User, time=0.1, value=0.9}
           {arriveTangent=0.1177, interpMode=RCIM_Cubic, leaveTangent=0.1177, tangentMode=RCTM_User, time=1, value=1}
          ]}}}
       CurveIndex=$Emitter.Reveal
       Scale Curve=&Lerp_Float(A=0.7, B=1, Alpha=$System.SizeRandom)
      )
     )
    )) {isSetParameters=true, module=SetVariables_#8}
   /SetVariables(Emitter.SpriteReveal:NiagaraFloat=&Multiply_Float(
     A=&Lerp_Float(
      A=&FloatFromCurve(FloatCurve=DataInterface() {propertyValues {Curve {keys [<at 0 0>,<at 0.6 1>]}}}, CurveIndex=$User.Reveal)
      B=&FloatFromCurve(FloatCurve=DataInterface() {propertyValues {Curve {keys [<at 0 0>,<at 0.95 1>]}}}, CurveIndex=$User.Reveal)
      Alpha=$System.SystemRandom
     )
     B=&Lerp_Float(A=1, B=1.1, Alpha=$Emitter.Random)
    )) {isSetParameters=true, module=SetVariables_#9}
   /SetVariables(Emitter.SpriteLength:NiagaraFloat=&Multiply_Float(
     A=&FloatFromCurve(
      FloatCurve=DataInterface() {propertyValues {Curve {keys [
          {arriveTangent=2.5813, interpMode=RCIM_Cubic, leaveTangent=2.5813, tangentMode=RCTM_User, time=0, value=0.1}
          {tangentMode=RCTM_User, time=0.5, value=1}
         ]}}}
      CurveIndex=$Engine.System.Age
     )
     B=&Multiply_Float(
      A=$Emitter.Initial.SpriteLength
      B=&FloatFromCurve(
       FloatCurve=DataInterface() {propertyValues {Curve {keys [
           {interpMode=RCIM_Cubic, leaveTangent=6.4155, tangentMode=RCTM_Break, time=0, value=0}
           {arriveTangent=0.1191, interpMode=RCIM_Cubic, leaveTangent=0.1191, tangentMode=RCTM_User, time=0.1, value=0.9}
           {arriveTangent=0.1177, interpMode=RCIM_Cubic, leaveTangent=0.1177, tangentMode=RCTM_User, time=1, value=1}
          ]}}}
       CurveIndex=$Emitter.SpriteReveal
       Scale Curve=&Lerp_Float(A=0.7, B=1, Alpha=$System.SizeRandom)
      )
     )
    )) {isSetParameters=true, module=SetVariables_#10}
  ]
  particleSpawnScript [
   &InitializeParticle(
    Color=&MaskLinearColorBySpawnGroup(Masked Linear Color=<rgba 1 1 1 1>, Passthrough Linear Color=<rgba 1 1 1 0.6>)
    Lifetime=&MaskFloatBySpawnGroup(Masked Float=$System.CurrentLoopDuration, Passthrough Float=&UniformRangedFloat(Minimum=0.4, Maximum=0.6))
    Position Offset=&MaskVectorBySpawnGroup(Masked Vector=<0 0 0>, Passthrough Vector=<0 -5 0>)
    Sprite Rotation Mode=::NewEnumerator2
    Sprite Size Mode=::NewEnumerator3
    Sprite UV Mode=::NewEnumerator4
    Uniform Sprite Size=&MaskFloatBySpawnGroup(Masked Float=10, Passthrough Float=&UniformRangedFloat(Minimum=10, Maximum=15))
    UsePositionOffset=true
    Write Material Random=true
   )
   &DynamicMaterialParameters(
    Index 0 Param 1=$Emitter.Random
    Module.Index0Param0Write=true
    Module.Index0Param1Write=true
    Module.Index0Param2Write=true
    Module.Index0Param3Write=true
   )
   /SetVariables(Particles.VisibilityTag=&SpawnGroup()) {isSetParameters=true, module=SetVariables_#11}
   &ShapeLocation(
    MaskSpawnGroup=false
    Offset Coordinate Space=::Local
    Offset Mode=::NewEnumerator0
    Override Seed=false
    Sphere Radius=&MaskFloatBySpawnGroup(Masked Float=0, Passthrough Float=0.1)
    UseOwnerScale=true
   ) {module=ShapeLocation001}
   &ApplyInitialForces()
   &AddVelocity(
    Cone Angle=5
    Cone Axis=<1 0.9 0>
    Module.UseConeFalloff=false
    Module.UseVelDistribution=false
    Override Seed=false
    Velocity Mode=::NewEnumerator2
    Velocity Speed=&MaskFloatBySpawnGroup(
     Masked Float=0
     Passthrough Float=&Multiply_Float(
      A=&RandomRangeFloat(Minimum=2, Maximum=250)
      B=&Lerp_Float(A=0.7, B=1.2, Alpha=$System.SizeRandom)
     )
    )
   )
  ]
  particleUpdateScript [
   &SolveForcesAndVelocity(Clamp Velocity=false, Limit Acceleration=false)
   &ParticleState()
   /SetVariables(Particles.ParticleRelativeLifetimeRibbons:NiagaraFloat=&ReturnNormalizedExecIndex()) {isSetParameters=true, module=SetVariables_#12}
   /SetVariables(Particles.OpacityMultiplierRibbons:NiagaraFloat=&FloatFromCurve(
     FloatCurve=DataInterface() {propertyValues {Curve {keys [
         {interpMode=RCIM_Cubic, tangentMode=RCTM_User, time=0, value=0}
         {arriveTangent=0.007, interpMode=RCIM_Cubic, leaveTangent=0.007, tangentMode=RCTM_User, time=0.5, value=1.5}
         {arriveTangent=-0.7513, interpMode=RCIM_Cubic, leaveTangent=-0.7513, tangentMode=RCTM_User, time=0.9, value=1}
         {arriveTangent=-0.9808, interpMode=RCIM_Cubic, leaveTangent=-0.9808, tangentMode=RCTM_User, time=1.8, value=0}
        ]}}}
     CurveIndex=&Add_Float(
      A=&Multiply_Float(A=$Particles.ParticleRelativeLifetimeRibbons, B=-1)
      B=&Multiply_Float(A=$Emitter.Reveal, B=2)
     )
    )) {isSetParameters=true, module=SetVariables_#13}
   /SetVariables(Particles.InterpolateCurveRibbons:NiagaraFloat=&FloatFromCurve(
     FloatCurve=DataInterface() {propertyValues {Curve {keys [{arriveTangent=-0.6503, time=0, value=0}, {arriveTangent=0.8679, time=1, value=1}]}}}
     CurveIndex=$Particles.ParticleRelativeLifetimeRibbons
    )) {isSetParameters=true, module=SetVariables_#14}
   /SetVariables(Particles.PositionRibbons:NiagaraPosition=&AddVectorToPosition(
     Position=&AddVectorToPosition(
      Position=&SimulationPosition()
      Vector=&TransformVector(Vector=<0 -5 0>, Destination Space=::Simulation)
     )
     Vector=&VectorFromFloat(Value=&Abs_Float(Float:NiagaraFloat=&NormalizeFloat(Value=&Multiply_Float(A=$Emitter.Length, B=$Particles.InterpolateCurveRibbons), Min=-1, Max=1)))
    )) {isSetParameters=true, module=SetVariables_#15}
   /SetVariables(
    Particles.RibbonColor:LinearColor=&MultiplyLinearColor(
     Linear Color=&MultiplyLinearColor(
      Linear Color=&ColorFromCurve(
       ColorCurve=DataInterface() {propertyValues {
         AlphaCurve {keys [
           <at 0 0>
           <at 0.0346 1>
           <at 0.2133 0.8>
           {arriveTangent=-1.0175, interpMode=RCIM_Cubic, leaveTangent=-1.0175, time=0.7268, value=0.8}
           <at 0.9995 0 RCIM_Cubic>
          ]}
         BlueCurve {keys [<at 0.4438 1>]}
         GreenCurve {keys [<at 0.4438 1>]}
         RedCurve {keys [<at 0.4438 1>]}
        }}
       CurveIndex=$Particles.ParticleRelativeLifetimeRibbons
      )
      Scale Factor=&MakeVector4_FromLinearColor(LinearColor=&FNTimeOfDayColor(Night Color=<rgba 0.1771 0.2756 1 0.7>, Morning Color=<rgba 1 1 1 1>, Day Color=<rgba 1 1 1 1>, Evening Color=<rgba 1 0.8696 0.5989 1>))
     )
     Scale Factor=&MakeFloat4FromVectorAndFloat(
      "Vector (RGB)"=&VectorFromFloat(Value=&FloatFromCurve(FloatCurve=DataInterface() {propertyValues {Curve {keys [<at 0.45 1 RCIM_Cubic>]}}}, CurveIndex=$User.Reveal))
      "Float (Alpha)"=1
     )
    )
    Particles.RibbonWidth=&Multiply_Float(
     A=&FloatFromCurve(
      FloatCurve=DataInterface() {propertyValues {Curve {keys [
          {arriveTangent=0.675, interpMode=RCIM_Cubic, leaveTangent=0.675, tangentMode=RCTM_User, time=0, value=0.2}
          {arriveTangent=1.0523, interpMode=RCIM_Cubic, leaveTangent=1.0523, tangentMode=RCTM_User, time=0.2799, value=0.5408}
          {arriveTangent=-0.0372, interpMode=RCIM_Cubic, leaveTangent=-0.0372, tangentMode=RCTM_User, time=0.7, value=0.8}
          {arriveTangent=-1.5957, interpMode=RCIM_Cubic, leaveTangent=-1.5957, tangentMode=RCTM_User, time=1, value=0.5138}
         ]}}}
      CurveIndex=$Particles.InterpolateCurveRibbons
      Scale Curve=15
     )
     B=&FloatFromCurve(
      FloatCurve=DataInterface() {propertyValues {Curve {keys [
          {arriveTangent=5.9942, interpMode=RCIM_Cubic, leaveTangent=5.9942, tangentMode=RCTM_User, time=0, value=0.475}
          {arriveTangent=0.0898, interpMode=RCIM_Cubic, leaveTangent=0.0898, tangentMode=RCTM_User, time=0.3, value=1}
         ]}}}
      CurveIndex=$System.Age
     )
    )
   ) {isSetParameters=true, module=SetVariables_#16}
   /SetVariables(Particles.ScaledRibbonWidth:NiagaraFloat=&Multiply_Float(
     A=$Particles.RibbonWidth
     B=&Multiply_Float(
      A=&FloatFromCurve(
       FloatCurve=DataInterface() {propertyValues {Curve {keys [
           {interpMode=RCIM_Cubic, leaveTangent=0.5491, tangentMode=RCTM_Break, time=0, value=0.7}
           {arriveTangent=1.0086, interpMode=RCIM_Cubic, leaveTangent=1.0086, tangentMode=RCTM_User, time=0.3931, value=1}
           {arriveTangent=0.0216, interpMode=RCIM_Cubic, tangentMode=RCTM_Break, time=1, value=1.5}
          ]}}}
       CurveIndex=$Emitter.Reveal
       Scale Curve=&Lerp_Float(
        A=1.6
        B=&FloatFromCurve(
         FloatCurve=DataInterface() {propertyValues {Curve {keys [<at 0 1.9 RCIM_Constant>,<at 0.333 2.25 RCIM_Constant>,<at 0.666 2.6 RCIM_Constant>]}}}
         CurveIndex=$Emitter.Random
        )
        Alpha=$System.SizeRandom
       )
      )
      B=1
     )
    )) {isSetParameters=true, module=SetVariables_#17}
   /SetVariables(Particles.ScaledRibbonColor:LinearColor=&Multiply_LinearColorByFloat(
     Linear Color=&MultiplyLinearColor(
      Linear Color=$Particles.RibbonColor
      Scale Factor=&MakeFloat4FromVectorAndFloat(
       "Vector (RGB)"=<1 1 1>
       "Float (Alpha)"=&Multiply_Float(
        A=$Particles.OpacityMultiplierRibbons
        B=&Multiply_Float(
         A=&FloatFromCurve(
          FloatCurve=DataInterface() {propertyValues {Curve {keys [
              {interpMode=RCIM_Cubic, tangentMode=RCTM_Break, time=0.6, value=1}
              {arriveTangent=-4.2498, interpMode=RCIM_Cubic, tangentMode=RCTM_Break, time=0.99, value=0}
             ]}}}
          CurveIndex=$Emitter.Reveal
          Scale Curve=&Lerp_Float(A=0.8, B=1, Alpha=$Emitter.Random)
         )
         B=$System.OpacityRandom
        )
       )
      )
     )
     Float=&FloatFromCurve(
      FloatCurve=DataInterface() {propertyValues {Curve {keys [
          {arriveTangent=5.7153, interpMode=RCIM_Cubic, leaveTangent=5.7153, tangentMode=RCTM_User, time=0, value=0}
          {arriveTangent=-1.4405, interpMode=RCIM_Cubic, leaveTangent=-1.4405, tangentMode=RCTM_User, time=0.3, value=2}
          {arriveTangent=-1.4711, interpMode=RCIM_Cubic, leaveTangent=-1.4711, tangentMode=RCTM_User, time=1, value=0}
         ]}}}
      CurveIndex=$System.Age
     )
    )) {isSetParameters=true, module=SetVariables_#18}
   &CameraOffset(Camera Offset Amount=5) {enabled=false}
   &ParticleState() {module=ParticleState001}
   &ScaleSpriteSize(Scale Factor=&MaskVector2dBySpawnGroup(
     Masked Vector2D=$Particles.Initial.SpriteSize
     Passthrough Vector2D=&Multiply_Vector2D_ByFloat(
      Float=&FloatFromCurve(
       FloatCurve=DataInterface() {propertyValues {Curve {keys [
           {arriveTangent=5.7302, interpMode=RCIM_Cubic, leaveTangent=5.7302, tangentMode=RCTM_User, time=0, value=0}
           {arriveTangent=0.2869, interpMode=RCIM_Cubic, leaveTangent=0.2869, tangentMode=RCTM_User, time=0.5, value=0.975}
          ]}}}
       CurveIndex=$Engine.System.Age
      )
      Vector2D=&Vector2DFromFloat(Value=&FloatFromCurve(
        FloatCurve=DataInterface() {propertyValues {Curve {keys [{tangentMode=RCTM_User, time=0, value=0.4}, <at 0.1 0.1>, {tangentMode=RCTM_User, time=1, value=1}]}}}
        Scale Curve=&Multiply_Float(A=0.7, B=&Lerp_Float(A=0.7, B=1, Alpha=$System.SizeRandom))
       ))
     )
     Spawn Group Mask=0
    )) {module=ScaleSpriteSize001}
   &SpriteRotationRate(Rotation Rate=&Multiply_Float(
     A=&RandomRangeFloat(Minimum=100, Maximum=150)
     B=&UniformAOrBFloat(A=-1, Evaluation Type=::NewEnumerator0, Recalculate Random Each Loop=true, Override Seed=false)
    ))
   /SetVariables(Particles.ParticleRelativeLifetimeSprites:NiagaraFloat=&Frac_Float(Float=&Add_Float(A=&Multiply_Float(A=1.5, B=$System.Age), B=&ReturnNormalizedExecIndex()))) {isSetParameters=true, module=SetVariables_#19}
   /SetVariables(Particles.OpacityMultiplierSprites:NiagaraFloat=&FloatFromCurve(
     FloatCurve=DataInterface() {propertyValues {Curve {keys [
         {interpMode=RCIM_Cubic, tangentMode=RCTM_User, time=0, value=0}
         {arriveTangent=-0.6677, interpMode=RCIM_Cubic, leaveTangent=-0.6677, tangentMode=RCTM_User, time=0.2, value=1.5}
         {arriveTangent=-0.6675, tangentMode=RCTM_User, time=1.5, value=0}
        ]}}}
     CurveIndex=&Add_Float(
      A=&Multiply_Float(A=$Particles.ParticleRelativeLifetimeSprites, B=-1)
      B=&Multiply_Float(A=$Emitter.SpriteReveal, B=&RandomRangeFloat(Minimum=1, Maximum=2))
     )
    )) {isSetParameters=true, module=SetVariables_#20}
   /SetVariables(Particles.InterpolateCurveSprites:NiagaraFloat=&FloatFromCurve(
     FloatCurve=DataInterface() {propertyValues {Curve {keys [
         {arriveTangent=-0.6503, interpMode=RCIM_Cubic, leaveTangent=1.4736, tangentMode=RCTM_Break, time=0, value=0}
         {arriveTangent=0.4202, interpMode=RCIM_Cubic, tangentMode=RCTM_Break, time=1, value=1}
        ]}}}
     CurveIndex=$Particles.ParticleRelativeLifetimeSprites
    )) {isSetParameters=true, module=SetVariables_#21}
   /SetVariables(Particles.PositionSprites:NiagaraPosition=&AddVectorToPosition(
     Position=&AddVectorToPosition(
      Position=&SimulationPosition()
      Vector=&TransformVector(Vector=<0 -5 0>, Destination Space=::Simulation)
     )
     Vector=&VectorFromFloat(Value=&Abs_Float(Float:NiagaraFloat=&NormalizeFloat(Value=&Multiply_Float(A=$Emitter.SpriteLength, B=$Particles.InterpolateCurveSprites), Min=-1, Max=1)))
    )) {isSetParameters=true, module=SetVariables_#22}
   /SetVariables(Particles.SpriteSize=&MaskVector2dBySpawnGroup(
     Masked Vector2D=&Multiply_Vector2D_ByFloat(
      Float=&FloatFromCurve(
       FloatCurve=DataInterface() {propertyValues {Curve {keys [
           {arriveTangent=4.522, interpMode=RCIM_Cubic, leaveTangent=4.522, tangentMode=RCTM_User, time=0, value=0.475}
           {arriveTangent=0.0898, interpMode=RCIM_Cubic, leaveTangent=0.0898, tangentMode=RCTM_User, time=0.3, value=1}
          ]}}}
       CurveIndex=$System.Age
      )
      Vector2D=&Multiply_Vector2D(
       A=&Vector2DFromFloat(Value=&FloatFromCurve(
         FloatCurve=DataInterface() {propertyValues {Curve {keys [{tangentMode=RCTM_User, time=0, value=0}, {tangentMode=RCTM_User, time=1, value=1}]}}}
         CurveIndex=$Particles.InterpolateCurveSprites
         Scale Curve=&Lerp_Float(A=35, B=50, Alpha=&RandomRangeFloat(Minimum=0, Maximum=1))
        ))
       B=&Vector2DFromFloat(Value=&Lerp_Float(A=0.7, B=1, Alpha=$System.SizeRandom))
      )
     )
     Passthrough Vector2D=$Particles.Initial.SpriteSize
     Spawn Group Mask=0
    )) {isSetParameters=true, module=SetVariables_#23}
   &SubUVAnimation(Start Frame=20, SubUV Lookup Index:NiagaraFloat=$Particles.ParticleRelativeLifetimeSprites)
   &SubUVAnimation(
    Output Alternate Renderer Binding=true
    "Sprite Renderer "=DataInterface() {propertyValues {SpriteRenderer=&self:MainBurst001_1.NiagaraSpriteRendererProperties_1}}
    Start Frame Range Override=40
    UseEndFrame=true
    UseStartFrame=true
   ) {module=SubUVAnimation002}
   &ScaleColor(
    Scale Alpha=&MaskFloatBySpawnGroup(
     Masked Float=&Multiply_Float(
      A=&Multiply_Float(
       A=&FloatFromCurve(
        FloatCurve=DataInterface() {propertyValues {Curve {keys [
            {arriveTangent=16.5612, interpMode=RCIM_Cubic, leaveTangentWeight=0.3333, time=0, value=0}
            {arriveTangentWeight=0.1333, interpMode=RCIM_Cubic, leaveTangentWeight=0.1667, tangentMode=RCTM_User, time=0.333, value=1}
            {arriveTangentWeight=0.1667, interpMode=RCIM_Cubic, tangentMode=RCTM_User, time=1, value=0}
           ]}}}
        CurveIndex=$Particles.ParticleRelativeLifetimeSprites
        Scale Curve=&RandomRangeFloat(Minimum=0.5, Maximum=1)
       )
       B=&Multiply_Float(
        A=&FloatFromCurve(
         FloatCurve=DataInterface() {propertyValues {Curve {keys [
             {interpMode=RCIM_Cubic, tangentMode=RCTM_Break, time=0.6, value=1}
             {arriveTangent=-4.2498, interpMode=RCIM_Cubic, tangentMode=RCTM_Break, time=0.99, value=0}
            ]}}}
         CurveIndex=$Emitter.SpriteReveal
         Scale Curve=&Lerp_Float(A=1, B=2, Alpha=$Emitter.Random)
        )
        B=$System.OpacityRandom
       )
      )
      B=$Particles.OpacityMultiplierSprites
     )
     Passthrough Float=&Multiply_Float(
      A=&FloatFromCurve(
       FloatCurve=DataInterface() {propertyValues {Curve {keys [
           {arriveTangent=5.5857, interpMode=RCIM_Cubic, leaveTangent=5.5857, leaveTangentWeight=0.3333, tangentMode=RCTM_User, time=0, value=0.3}
           {arriveTangentWeight=0.1333, interpMode=RCIM_Cubic, leaveTangentWeight=0.1667, tangentMode=RCTM_User, time=0.1, value=2}
           {arriveTangentWeight=0.1667, interpMode=RCIM_Cubic, tangentMode=RCTM_User, time=1, value=0}
          ]}}}
       Scale Curve=&Multiply_Float(A=1, B=$System.OpacityRandom)
      )
      B=&RandomRangeFloat(Minimum=0.5, Maximum=1)
     )
    )
    Scale RGB=&MakeVectorFromLinearColorRGB(Color=&FNTimeOfDayColor(
      Night Color=<rgba 0.1771 0.2756 1 0.7>
      Morning Color=<rgba 0.5607 0.6584 1 1>
      Day Color=<rgba 1 1 1 1>
      Evening Color=<rgba 1 0.8696 0.5989 1>
     ))
    ScaleA=true
    ScaleRGB=true
   )
   &ScaleColor(
    Scale Alpha=&MaskFloatBySpawnGroup(
     Masked Float=&Multiply_Float(
      A=$Particles.OpacityMultiplierSprites
      B=&FloatFromCurve(
       FloatCurve=DataInterface() {propertyValues {Curve {keys [
           {arriveTangent=9.0725, interpMode=RCIM_Cubic, leaveTangent=9.0725, tangentMode=RCTM_User, time=0, value=0}
           {arriveTangent=-0.6131, interpMode=RCIM_Cubic, leaveTangent=-0.6131, tangentMode=RCTM_User, time=0.1408, value=0.9847}
           {arriveTangent=-1.4711, interpMode=RCIM_Cubic, leaveTangent=-1.4711, tangentMode=RCTM_User, time=1, value=0}
          ]}}}
       CurveIndex=$Emitter.Age
      )
     )
     Passthrough Float=&FloatFromCurve(
      FloatCurve=DataInterface() {propertyValues {Curve {keys [
          {arriveTangent=9.0725, interpMode=RCIM_Cubic, leaveTangent=9.0725, tangentMode=RCTM_User, time=0, value=0}
          {arriveTangent=-2.5249, interpMode=RCIM_Cubic, leaveTangent=-2.5249, tangentMode=RCTM_User, time=0.1408, value=0.9847}
          {arriveTangent=-0.0405, interpMode=RCIM_Cubic, leaveTangent=-0.0405, tangentMode=RCTM_User, time=1, value=0}
         ]}}}
      CurveIndex=$Emitter.Age
     )
    )
    ScaleA=true
    ScaleRGB=false
   ) {module=ScaleColor001}
   &ScaleColor(
    Scale Alpha=&MaskFloatBySpawnGroup(
     Masked Float=&Multiply_Float(
      A=$Particles.OpacityMultiplierSprites
      B=&FloatFromCurve(
       FloatCurve=DataInterface() {propertyValues {Curve {keys [
           {arriveTangent=9.0725, interpMode=RCIM_Cubic, leaveTangent=9.0725, tangentMode=RCTM_User, time=0, value=0}
           {arriveTangent=-0.6131, interpMode=RCIM_Cubic, leaveTangent=-0.6131, tangentMode=RCTM_User, time=0.1408, value=0.9847}
           {arriveTangent=-1.4711, interpMode=RCIM_Cubic, leaveTangent=-1.4711, tangentMode=RCTM_User, time=1, value=0}
          ]}}}
       CurveIndex=$Emitter.Age
      )
     )
     Passthrough Float=&FloatFromCurve(
      FloatCurve=DataInterface() {propertyValues {Curve {keys [
          {arriveTangent=9.0725, interpMode=RCIM_Cubic, leaveTangent=9.0725, tangentMode=RCTM_User, time=0, value=0}
          {arriveTangent=-2.5249, interpMode=RCIM_Cubic, leaveTangent=-2.5249, tangentMode=RCTM_User, time=0.1408, value=0.9847}
          {arriveTangent=-0.0405, interpMode=RCIM_Cubic, leaveTangent=-0.0405, tangentMode=RCTM_User, time=1, value=0}
         ]}}}
      Scale Curve=&Multiply_Float(A=1, B=$System.ToD)
     )
    )
    ScaleA=true
    ScaleRGB=false
   ) {module=ScaleColor002}
  ]
  rendererClasses [&NiagaraRibbonRendererProperties, &NiagaraSpriteRendererProperties]
  renderers [
   {rendererClass=&NiagaraRibbonRendererProperties}
   {rendererClass=&NiagaraSpriteRendererProperties, rendererIndex=1}
   {rendererClass=&NiagaraSpriteRendererProperties, rendererIndex=2}
  ]
 }
}
summary {
 systemName=NS_Emote_CrumbViolin_Steam
 emitters [
  {emitterName=MainBurst, rendererClasses [&NiagaraRibbonRendererProperties, &NiagaraSpriteRendererProperties]}
  {
   bEnabled=false
   emitterName=MainBurst001
   rendererClasses [&NiagaraRibbonRendererProperties, &NiagaraSpriteRendererProperties]
  }
 ]
 userVariables [
  User.bIsFrontEnd {type=NiagaraBool}
  User.bIsFrontEndPreview {type=NiagaraBool}
  User.bIsLargeBodyPawn {type=NiagaraBool}
  User.Reveal {type=NiagaraFloat, defaultValue=0.5, description=""}
 ]
}
