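# Niagara system for the CrumbViolin emote that puffs animated steam sprites from selected bone or socket points on a skeletal mesh.
# Early and late burst emitters bracket SteamBursts, which fires two probabilistic waves, each combining bursts with curve-driven spawn rates.
# Particle partitions index into integer arrays to choose attachment points, resampled in update with small spherical offsets.
# Curves drive drag, growth and fading, with time-of-day tinting and per-system opacity variation.
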NiagaraSystem /BRCosmetics/Animation/Game/MainPlayer/Emotes/CrumbViolin/FX/NS_Emote_CrumbViolin_SteamBurst
dependencies {
 usedDataInterfaces [
  &NiagaraDataInterfaceCurve
  &NiagaraDataInterfaceArrayInt32
  &NiagaraDataInterfaceColorCurve
  &NiagaraDataInterfaceSkeletalMesh
 ]
 usedDynamicInputs [
  &SimulationPosition
  &RandomRangeVector
  &Lerp_Vector
  &MakeFloatFromInt
  &FNTimeOfDayScalar
  &ClampFloat
 ]
 usedModules [
  &self:SystemSpawnScript.SystemScriptSource.SystemScriptGraph.NiagaraNodeAssignment_0.SetVariables_#1
  &self:SystemSpawnScript.SystemScriptSource.SystemScriptGraph.NiagaraNodeAssignment_3.SetVariables_#2
  &self:SystemSpawnScript.SystemScriptSource.SystemScriptGraph.NiagaraNodeAssignment_4.SetVariables_#3
  &SystemState
  &self:SystemSpawnScript.SystemScriptSource.SystemScriptGraph.NiagaraNodeAssignment_1.SetVariables_#4
  &InterpolateOverTime
  &self:SystemSpawnScript.SystemScriptSource.SystemScriptGraph.NiagaraNodeAssignment_2.SetVariables_#5
 ]
}
emitters {
 Single_Burst_Early {
  emitterUpdateScript [
   &EmitterState(Life Cycle Mode=::NewEnumerator1, Loop Behavior=::NewEnumerator1, Loop Duration=0.5, UseLoopDelay=true)
   &SpawnBurst_Instantaneous(
    Spawn Count=&RandomRangeInt(Minimum=15, Maximum=17)
    Spawn Time=&RandomRangeFloat(Minimum=0, Maximum=0.5, Evaluation Type=::NewEnumerator1)
    Use Spawn Probability=false
    UseLoopCountLimit=false
   )
   &SpawnRate(
    SpawnRate=&FloatFromCurve(
     FloatCurve=DataInterface() {propertyValues {Curve {keys [
         {interpMode=RCIM_Cubic, tangentMode=RCTM_User, time=0, value=1}
         {arriveTangent=-2.036, interpMode=RCIM_Cubic, leaveTangent=-2.036, tangentMode=RCTM_User, time=1, value=0.125}
        ]}}}
     CurveIndex=$Emitter.NormalizedLoopAge
     Scale Curve=100
    )
    Use Spawn Probability=false
   )
  ]
  particleSpawnScript [
   /SetVariables(Particles.RibbonID=/Niagara/DynamicInputs/Meshes/MakeRibbonIDFromSpawnGroup()) {isSetParameters=true, module=SetVariables_#6}
   &PartitionParticles(
    Number Of Partitions=4
    Particles Per Partition:NiagaraInt32=&RandomRangeInt(Minimum=15, Maximum=17)
    Partition Mode:ENiagara_PartitionMode=::NewEnumerator1
    Write Partition to Ribbon ID=::NewEnumerator0
   )
   &InitializeParticle(
    Color=<rgba 1 1 1 0.6>
    Lifetime=&UniformRangedFloat(Minimum=0.4, Maximum=0.55)
    Mass Mode=::NewEnumerator1
    Position Offset=<0 -5 0>
    Sprite Rotation Mode=::NewEnumerator2
    Sprite Size=&Vector2DFromFloat(Value=&UniformRangedFloat(Minimum=25, Maximum=45))
    Sprite Size Mode=::NewEnumerator4
    UsePositionOffset=true
    Write Material Random=true
   ) {module=InitializeParticle001}
   &ApplyInitialForces()
   &AddVelocity(
    Cone Angle=5
    Cone Axis=$Particles.SkeletalMeshLocation.SampledPosition
    Module.UseConeFalloff=false
    Module.UseVelDistribution=false
    Override Seed=false
    Velocity Mode=::NewEnumerator2
    Velocity Speed=&Multiply_Float(
     A=&RandomRangeFloat(Minimum=2, Maximum=75)
     B=&Lerp_Float(A=0.7, B=1, Alpha=$System.SizeRandom)
    )
   ) {enabled=false}
  ]
  particleUpdateScript [
   &SkeletalMeshLocation(
    Bone / Socket Index=&SelectIntFromArray(Direct Array Index=$Particles.PartitionParticles.Partition, Int Selection Array=DataInterface() {propertyValues {IntData [6,8,15,9]}})
    Bone Sampling Mode=::NewEnumerator4
    MaskSpawnGroup=false
    Position Sampling=::NewEnumerator1
    Skeletal Mesh=$System.Skeletal Mesh
   )
   &ShapeLocation(
    MaskSpawnGroup=false
    Offset Coordinate Space=::Local
    Offset Mode=::NewEnumerator0
    Override Seed=false
    Shape Origin=$StackContext.SkeletalMeshLocation.SampledPosition
    Sphere Radius=5
    UseOwnerScale=true
   ) {module=ShapeLocation001}
   &LinearForce(Force=&VectorFromFloat(Value=&Multiply_Float(
      A=&RandomRangeFloat(Minimum=2, Maximum=75)
      B=&Lerp_Float(A=0.7, B=1, Alpha=$System.SizeRandom)
     )))
   &ParticleState()
   &SubUVAnimation(Start Frame=40, SubUV Lookup Index:NiagaraFloat=&Multiply_Float(A=$Particles.NormalizedAge, B=1))
   &SpriteRotationRate(Rotation Rate=&Multiply_Float(
     A=&RandomRangeFloat(Minimum=100, Maximum=150)
     B=&UniformAOrBFloat(A=-1, Evaluation Type=::NewEnumerator0, Recalculate Random Each Loop=true, Override Seed=false)
    ))
   &Drag(
    Drag=&FloatFromCurve(
     FloatCurve=DataInterface() {propertyValues {Curve {keys [{interpMode=RCIM_Cubic, tangentMode=RCTM_User, time=0, value=0}, {interpMode=RCIM_Cubic, tangentMode=RCTM_User, time=1, value=1}]}}}
     Scale Curve=7
    )
    Use Linear Drag=true
    Use Rotational Drag=false
   )
   &SolveForcesAndVelocity(Clamp Velocity=false, Limit Acceleration=false)
   &ScaleColor(
    Scale Alpha=&Multiply_Float(
     A=&FloatFromCurve(
      FloatCurve=DataInterface() {propertyValues {Curve {keys [
          {arriveTangent=5.5857, interpMode=RCIM_Cubic, leaveTangent=5.5857, leaveTangentWeight=0.3333, tangentMode=RCTM_User, time=0, value=0.3}
          {arriveTangentWeight=0.1333, interpMode=RCIM_Cubic, leaveTangentWeight=0.1667, tangentMode=RCTM_User, time=0.1, value=1.7}
          {arriveTangentWeight=0.1667, interpMode=RCIM_Cubic, tangentMode=RCTM_User, time=1, value=0}
         ]}}}
      Scale Curve=&Multiply_Float(A=$System.ToD, B=$System.OpacityRandom)
     )
     B=&RandomRangeFloat(Minimum=0.5, Maximum=1)
    )
    Scale RGB=&Multiply_VectorByFloat(
     Vector=&MakeVectorFromLinearColorRGB(Color=&MultiplyLinearColor(
       Linear Color=&ColorFromCurve(ColorCurve=DataInterface() {propertyValues {
          AlphaCurve {keys [
            <at 0 0>
            <at 0.0346 1>
            <at 0.2133 0.8>
            {arriveTangent=-1.0175, interpMode=RCIM_Cubic, leaveTangent=-1.0175, time=0.7268, value=0.8}
            <at 0.9995 0 RCIM_Cubic>
           ]}
          BlueCurve {keys [<at 0.4438 1>]}
          GreenCurve {keys [<at 0.4438 1>]}
          RedCurve {keys [<at 0.4438 1>]}
         }})
       Scale Factor=&MakeVector4_FromLinearColor(LinearColor=&FNTimeOfDayColor(
         Night Color=<rgba 0.5104 0.569 1 1>
         Morning Color=<rgba 0.6875 0.7565 1 1>
         Day Color=<rgba 1 1 1 1>
         Evening Color=<rgba 1 0.8696 0.5989 1>
        ))
      ))
     Float=&FloatFromCurve(
      FloatCurve=DataInterface() {propertyValues {Curve {keys [
          {arriveTangent=9.0725, interpMode=RCIM_Cubic, leaveTangent=9.0725, tangentMode=RCTM_User, time=0, value=0}
          {arriveTangent=-2.5249, interpMode=RCIM_Cubic, leaveTangent=-2.5249, tangentMode=RCTM_User, time=0.1408, value=2}
          {arriveTangent=-0.0405, interpMode=RCIM_Cubic, leaveTangent=-0.0405, tangentMode=RCTM_User, time=1, value=0}
         ]}}}
      Scale Curve=&RandomRangeFloat(Minimum=0.9, Maximum=1)
     )
    )
    ScaleA=true
    ScaleRGB=true
   )
   &ScaleSpriteSize(Scale Factor=&Multiply_Vector2D_ByFloat(
     Float=&FloatFromCurve(
      FloatCurve=DataInterface() {propertyValues {Curve {keys [
          {arriveTangent=1.6669, interpMode=RCIM_Cubic, leaveTangent=1.6669, tangentMode=RCTM_User, time=0, value=0}
          {tangentMode=RCTM_User, time=1, value=1}
         ]}}}
      CurveIndex=$Engine.System.Age
     )
     Vector2D=&Vector2DFromFloat(Value=&FloatFromCurve(
       FloatCurve=DataInterface() {propertyValues {Curve {keys [<at 0 0.1 RCIM_Cubic>, {arriveTangent=3.4967, tangentMode=RCTM_User, time=1, value=2.8}]}}}
       Scale Curve=&Multiply_Float(A=0.7, B=&Lerp_Float(A=0.5, B=1, Alpha=$System.SizeRandom))
      ))
    )) {module=ScaleSpriteSize001}
   &CameraOffset(Camera Offset Amount=&RandomRangeFloat(Minimum=1, Maximum=15))
  ]
  rendererClasses [&NiagaraSpriteRendererProperties]
  renderers [{rendererClass=&NiagaraSpriteRendererProperties}]
 }
 Single_Burst_Late {
  emitterUpdateScript [
   &EmitterState(Life Cycle Mode=::NewEnumerator1, Loop Behavior=::NewEnumerator1, Loop Delay=4.3, Loop Duration=0.5, UseLoopDelay=true)
   &SpawnRate(
    SpawnRate=&FloatFromCurve(
     FloatCurve=DataInterface() {propertyValues {Curve {keys [
         {interpMode=RCIM_Cubic, tangentMode=RCTM_User, time=0, value=1}
         {arriveTangent=-2.036, interpMode=RCIM_Cubic, leaveTangent=-2.036, tangentMode=RCTM_User, time=1, value=0.125}
        ]}}}
     CurveIndex=$Emitter.NormalizedLoopAge
     Scale Curve=95
    )
    Use Spawn Probability=false
   )
   &SpawnBurst_Instantaneous(
    Spawn Count=&RandomRangeInt(Minimum=15, Maximum=17)
    Spawn Time=&RandomRangeFloat(Minimum=0, Maximum=0.2, Evaluation Type=::NewEnumerator1)
    Use Spawn Probability=false
    UseLoopCountLimit=false
   )
  ]
  particleSpawnScript [
   /SetVariables(Particles.RibbonID=&MakeRibbonIDFromSpawnGroup()) {isSetParameters=true, module=SetVariables_#6}
   &PartitionParticles(
    Particles Per Partition:NiagaraInt32=&RandomRangeInt(Minimum=15, Maximum=17)
    Partition Mode:ENiagara_PartitionMode=::NewEnumerator1
    Write Partition to Ribbon ID=::NewEnumerator0
   )
   &InitializeParticle(
    Color=<rgba 1 1 1 0.6>
    Lifetime=&UniformRangedFloat(Minimum=0.4, Maximum=0.55)
    Mass Mode=::NewEnumerator1
    Position Offset=<0 -5 0>
    Sprite Rotation Mode=::NewEnumerator2
    Sprite Size=&Vector2DFromFloat(Value=&UniformRangedFloat(Minimum=25, Maximum=45))
    Sprite Size Mode=::NewEnumerator4
    UsePositionOffset=true
    Write Material Random=true
   ) {module=InitializeParticle001}
   &ApplyInitialForces()
   &AddVelocity(
    Cone Angle=5
    Cone Axis=$Particles.SkeletalMeshLocation.SampledPosition
    Module.UseConeFalloff=false
    Module.UseVelDistribution=false
    Override Seed=false
    Velocity Mode=::NewEnumerator2
    Velocity Speed=&Multiply_Float(
     A=&RandomRangeFloat(Minimum=2, Maximum=75)
     B=&Lerp_Float(A=0.7, B=1, Alpha=$System.SizeRandom)
    )
   ) {enabled=false}
  ]
  particleUpdateScript [
   &SkeletalMeshLocation(
    Bone / Socket Index=&SelectIntFromArray(Direct Array Index=$Particles.PartitionParticles.Partition, Int Selection Array=DataInterface() {propertyValues {IntData [7,11]}})
    Bone Sampling Mode=::NewEnumerator4
    MaskSpawnGroup=false
    Position Sampling=::NewEnumerator1
    Skeletal Mesh=$System.Skeletal Mesh
   )
   &ShapeLocation(
    MaskSpawnGroup=false
    Offset Coordinate Space=::Local
    Offset Mode=::NewEnumerator0
    Override Seed=false
    Shape Origin=$StackContext.SkeletalMeshLocation.SampledPosition
    Sphere Radius=5
    UseOwnerScale=true
   ) {module=ShapeLocation001}
   &LinearForce(Force=&VectorFromFloat(Value=&Multiply_Float(
      A=&RandomRangeFloat(Minimum=2, Maximum=75)
      B=&Lerp_Float(A=0.7, B=1, Alpha=$System.SizeRandom)
     )))
   &ParticleState()
   &SubUVAnimation(Start Frame=40, SubUV Lookup Index:NiagaraFloat=&Multiply_Float(A=$Particles.NormalizedAge, B=1))
   &SpriteRotationRate(Rotation Rate=&Multiply_Float(
     A=&RandomRangeFloat(Minimum=100, Maximum=150)
     B=&UniformAOrBFloat(A=-1, Evaluation Type=::NewEnumerator0, Recalculate Random Each Loop=true, Override Seed=false)
    ))
   &Drag(
    Drag=&FloatFromCurve(
     FloatCurve=DataInterface() {propertyValues {Curve {keys [{interpMode=RCIM_Cubic, tangentMode=RCTM_User, time=0, value=0}, {interpMode=RCIM_Cubic, tangentMode=RCTM_User, time=1, value=1}]}}}
     Scale Curve=7
    )
    Use Linear Drag=true
    Use Rotational Drag=false
   )
   &SolveForcesAndVelocity(Clamp Velocity=false, Limit Acceleration=false)
   &ScaleColor(
    Scale Alpha=&Multiply_Float(
     A=&FloatFromCurve(
      FloatCurve=DataInterface() {propertyValues {Curve {keys [
          {arriveTangent=5.5857, interpMode=RCIM_Cubic, leaveTangent=5.5857, leaveTangentWeight=0.3333, tangentMode=RCTM_User, time=0, value=0.3}
          {arriveTangentWeight=0.1333, interpMode=RCIM_Cubic, leaveTangentWeight=0.1667, tangentMode=RCTM_User, time=0.1, value=1.7}
          {arriveTangentWeight=0.1667, interpMode=RCIM_Cubic, tangentMode=RCTM_User, time=1, value=0}
         ]}}}
      Scale Curve=&Multiply_Float(A=$System.ToD, B=$System.OpacityRandom)
     )
     B=&RandomRangeFloat(Minimum=0.5, Maximum=1)
    )
    Scale RGB=&Multiply_VectorByFloat(
     Vector=&MakeVectorFromLinearColorRGB(Color=&MultiplyLinearColor(
       Linear Color=&ColorFromCurve(ColorCurve=DataInterface() {propertyValues {
          AlphaCurve {keys [
            <at 0 0>
            <at 0.0346 1>
            <at 0.2133 0.8>
            {arriveTangent=-1.0175, interpMode=RCIM_Cubic, leaveTangent=-1.0175, time=0.7268, value=0.8}
            <at 0.9995 0 RCIM_Cubic>
           ]}
          BlueCurve {keys [<at 0.4438 1>]}
          GreenCurve {keys [<at 0.4438 1>]}
          RedCurve {keys [<at 0.4438 1>]}
         }})
       Scale Factor=&MakeVector4_FromLinearColor(LinearColor=&FNTimeOfDayColor(
         Night Color=<rgba 0.5104 0.569 1 1>
         Morning Color=<rgba 0.6875 0.7565 1 1>
         Day Color=<rgba 1 1 1 1>
         Evening Color=<rgba 1 0.8696 0.5989 1>
        ))
      ))
     Float=&FloatFromCurve(
      FloatCurve=DataInterface() {propertyValues {Curve {keys [
          {arriveTangent=9.0725, interpMode=RCIM_Cubic, leaveTangent=9.0725, tangentMode=RCTM_User, time=0, value=0}
          {arriveTangent=-2.5249, interpMode=RCIM_Cubic, leaveTangent=-2.5249, tangentMode=RCTM_User, time=0.1408, value=2}
          {arriveTangent=-0.0405, interpMode=RCIM_Cubic, leaveTangent=-0.0405, tangentMode=RCTM_User, time=1, value=0}
         ]}}}
      Scale Curve=&RandomRangeFloat(Minimum=0.9, Maximum=1)
     )
    )
    ScaleA=true
    ScaleRGB=true
   )
   &ScaleSpriteSize(Scale Factor=&Multiply_Vector2D_ByFloat(
     Float=&FloatFromCurve(
      FloatCurve=DataInterface() {propertyValues {Curve {keys [
          {arriveTangent=1.6669, interpMode=RCIM_Cubic, leaveTangent=1.6669, tangentMode=RCTM_User, time=0, value=0}
          {tangentMode=RCTM_User, time=1, value=1}
         ]}}}
      CurveIndex=$Engine.System.Age
     )
     Vector2D=&Vector2DFromFloat(Value=&FloatFromCurve(
       FloatCurve=DataInterface() {propertyValues {Curve {keys [<at 0 0.1 RCIM_Cubic>, {arriveTangent=3.4967, tangentMode=RCTM_User, time=1, value=2.8}]}}}
       Scale Curve=&Multiply_Float(A=0.6, B=&Lerp_Float(A=0.5, B=1, Alpha=$System.SizeRandom))
      ))
    )) {module=ScaleSpriteSize001}
   &CameraOffset(Camera Offset Amount=&RandomRangeFloat(Minimum=1, Maximum=15))
  ]
  rendererClasses [&NiagaraSpriteRendererProperties]
  renderers [{rendererClass=&NiagaraSpriteRendererProperties}]
 }
 SteamBursts {
  emitterUpdateScript [
   &EmitterState(Life Cycle Mode=::NewEnumerator1, Loop Behavior=::NewEnumerator1, Loop Duration=$System.CurrentLoopDuration, UseLoopDelay=false)
   &SpawnBurst_Instantaneous(
    Spawn Count=&RandomRangeInt(Minimum=15, Maximum=17)
    Spawn Group=1
    Spawn Probability=0.7
    Spawn Time=&RandomRangeFloat(Minimum=2.5, Maximum=2.7, Evaluation Type=::NewEnumerator1)
    Use Spawn Probability=true
    UseLoopCountLimit=false
   ) {module=SpawnBurst_Instantaneous001}
   &SpawnRate(
    Spawn Probability=0.7
    SpawnGroup=1
    SpawnRate=&FloatFromCurve(
     FloatCurve=DataInterface() {propertyValues {Curve {keys [
         <at 0 0 RCIM_Constant>
         <at 0.55 1 RCIM_Cubic>
         {arriveTangent=0.0102, interpMode=RCIM_Cubic, leaveTangent=0.0102, tangentMode=RCTM_User, time=0.9, value=0}
        ]}}}
     CurveIndex=$Emitter.NormalizedLoopAge
     Scale Curve=200
    )
    Use Spawn Probability=true
   ) {module=SpawnRate001}
   &SpawnRate(
    Spawn Probability=0.7
    SpawnRate=&FloatFromCurve(
     FloatCurve=DataInterface() {propertyValues {Curve {keys [
         {arriveTangent=4.587, interpMode=RCIM_Cubic, leaveTangent=4.587, tangentMode=RCTM_User, time=0, value=0.15}
         <at 0.4 1 RCIM_Constant>
         <at 0.5 0 RCIM_Constant>
        ]}}}
     CurveIndex=$Emitter.NormalizedLoopAge
     Scale Curve=55
    )
    Use Spawn Probability=true
   )
   &SpawnBurst_Instantaneous(
    Spawn Count=&RandomRangeInt(Minimum=15, Maximum=17)
    Spawn Probability=0.7
    Spawn Time=&RandomRangeFloat(Minimum=0, Maximum=0.2, Evaluation Type=::NewEnumerator1)
    Use Spawn Probability=true
    UseLoopCountLimit=false
   )
  ]
  particleSpawnScript [
   /SetVariables(Particles.RibbonID=&MakeRibbonIDFromSpawnGroup()) {isSetParameters=true, module=SetVariables_#6}
   &PartitionParticles(
    Normalized Index In Partition Includes One=true
    Number Of Partitions=&MaskIntBySpawnGroup(Masked Int=8, Passthrough Int=15, Spawn Group Mask=0)
    Particles Per Partition:NiagaraInt32=&RandomRangeInt(Minimum=15, Maximum=17)
    Partition Mode:ENiagara_PartitionMode=::NewEnumerator1
    Write Partition to Ribbon ID=::NewEnumerator0
   )
   &InitializeParticle(
    Color=<rgba 1 1 1 0.6>
    Lifetime=&UniformRangedFloat(Minimum=0.4, Maximum=0.55)
    Mass Mode=::NewEnumerator1
    Position Offset=<0 -5 0>
    Sprite Rotation Mode=::NewEnumerator2
    Sprite Size=&Vector2DFromFloat(Value=&UniformRangedFloat(Minimum=20, Maximum=35))
    Sprite Size Mode=::NewEnumerator4
    UsePositionOffset=true
    Write Material Random=true
   ) {module=InitializeParticle001}
   &ApplyInitialForces()
   &AddVelocity(
    Cone Angle=5
    Cone Axis=$Particles.SkeletalMeshLocation.SampledPosition
    Module.UseConeFalloff=false
    Module.UseVelDistribution=false
    Override Seed=false
    Velocity Mode=::NewEnumerator2
    Velocity Speed=&Multiply_Float(
     A=&RandomRangeFloat(Minimum=2, Maximum=75)
     B=&Lerp_Float(A=0.7, B=1, Alpha=$System.SizeRandom)
    )
   ) {enabled=false}
  ]
  particleUpdateScript [
   &SkeletalMeshLocation(
    Bone / Socket Index=&MaskIntBySpawnGroup(
     Masked Int=&SelectIntFromArray(
      Direct Array Index=$Particles.PartitionParticles.Partition
      Int Selection Array=DataInterface() {propertyValues {IntData [7,8,9,4,15,14,3,0,1]}}
     )
     Passthrough Int=&MaskIntBySpawnGroup(
      Masked Int=&SelectIntFromArray(
       Direct Array Index=$Particles.PartitionParticles.Partition
       Int Selection Array=DataInterface() {propertyValues {IntData [0,1,3,4,5,6,7,8,10,11,12,13,15]}}
      )
      Passthrough Int=$Particles.PartitionParticles.Partition
      Spawn Group Mask=0
     )
     Spawn Group Mask=0
    )
    Bone Sampling Mode=::NewEnumerator4
    MaskSpawnGroup=false
    Position Sampling=::NewEnumerator1
    Skeletal Mesh=$System.Skeletal Mesh
   )
   &ShapeLocation(
    MaskSpawnGroup=false
    Offset Coordinate Space=::Local
    Offset Mode=::NewEnumerator0
    Override Seed=false
    Shape Origin=$StackContext.SkeletalMeshLocation.SampledPosition
    Sphere Radius=5
    UseOwnerScale=true
   ) {module=ShapeLocation001}
   &ParticleState()
   &SubUVAnimation(Start Frame=40, SubUV Lookup Index:NiagaraFloat=&Multiply_Float(A=$Particles.NormalizedAge, B=1))
   &SpriteRotationRate(Rotation Rate=&Multiply_Float(
     A=&RandomRangeFloat(Minimum=100, Maximum=150)
     B=&UniformAOrBFloat(A=-1, Evaluation Type=::NewEnumerator0, Recalculate Random Each Loop=true, Override Seed=false)
    ))
   &Drag(
    Drag=&FloatFromCurve(
     FloatCurve=DataInterface() {propertyValues {Curve {keys [{interpMode=RCIM_Cubic, tangentMode=RCTM_User, time=0, value=0}, {interpMode=RCIM_Cubic, tangentMode=RCTM_User, time=1, value=1}]}}}
     Scale Curve=7
    )
    Use Linear Drag=true
    Use Rotational Drag=false
   )
   &LinearForce(Force=&VectorFromFloat(Value=&Multiply_Float(
      A=&RandomRangeFloat(Minimum=2, Maximum=75)
      B=&Lerp_Float(A=0.7, B=1, Alpha=$System.SizeRandom)
     )))
   &SolveForcesAndVelocity(Clamp Velocity=false, Limit Acceleration=false)
   &ScaleColor(
    Scale Alpha=&Multiply_Float(
     A=&FloatFromCurve(
      FloatCurve=DataInterface() {propertyValues {Curve {keys [
          {arriveTangent=5.5857, interpMode=RCIM_Cubic, leaveTangent=5.5857, leaveTangentWeight=0.3333, tangentMode=RCTM_User, time=0, value=0.3}
          {arriveTangentWeight=0.1333, interpMode=RCIM_Cubic, leaveTangentWeight=0.1667, tangentMode=RCTM_User, time=0.1, value=1.7}
          {arriveTangentWeight=0.1667, interpMode=RCIM_Cubic, tangentMode=RCTM_User, time=1, value=0}
         ]}}}
      Scale Curve=&Multiply_Float(A=$System.ToD, B=$System.OpacityRandom)
     )
     B=&RandomRangeFloat(Minimum=0.5, Maximum=1)
    )
    Scale RGB=&Multiply_VectorByFloat(
     Vector=&MakeVectorFromLinearColorRGB(Color=&MultiplyLinearColor(
       Linear Color=&ColorFromCurve(ColorCurve=DataInterface() {propertyValues {
          AlphaCurve {keys [
            <at 0 0>
            <at 0.0346 1>
            <at 0.2133 0.8>
            {arriveTangent=-1.0175, interpMode=RCIM_Cubic, leaveTangent=-1.0175, time=0.7268, value=0.8}
            <at 0.9995 0 RCIM_Cubic>
           ]}
          BlueCurve {keys [<at 0.4438 1>]}
          GreenCurve {keys [<at 0.4438 1>]}
          RedCurve {keys [<at 0.4438 1>]}
         }})
       Scale Factor=&MakeVector4_FromLinearColor(LinearColor=&FNTimeOfDayColor(
         Night Color=<rgba 0.5104 0.569 1 1>
         Morning Color=<rgba 0.6875 0.7565 1 1>
         Day Color=<rgba 1 1 1 1>
         Evening Color=<rgba 1 0.8696 0.5989 1>
        ))
      ))
     Float=&FloatFromCurve(
      FloatCurve=DataInterface() {propertyValues {Curve {keys [
          {arriveTangent=9.0725, interpMode=RCIM_Cubic, leaveTangent=9.0725, tangentMode=RCTM_User, time=0, value=0}
          {arriveTangent=-2.5249, interpMode=RCIM_Cubic, leaveTangent=-2.5249, tangentMode=RCTM_User, time=0.1408, value=2}
          {arriveTangent=-0.0405, interpMode=RCIM_Cubic, leaveTangent=-0.0405, tangentMode=RCTM_User, time=1, value=0}
         ]}}}
      Scale Curve=&RandomRangeFloat(Minimum=0.9, Maximum=1)
     )
    )
    ScaleA=true
    ScaleRGB=true
   )
   &ScaleSpriteSize(Scale Factor=&Multiply_Vector2D_ByFloat(
     Float=&FloatFromCurve(
      FloatCurve=DataInterface() {propertyValues {Curve {keys [
          {arriveTangent=1.6669, interpMode=RCIM_Cubic, leaveTangent=1.6669, tangentMode=RCTM_User, time=0, value=0}
          {tangentMode=RCTM_User, time=1, value=1}
         ]}}}
      CurveIndex=$Engine.System.Age
      Scale Curve=&RandomRangeFloat(Minimum=0.8, Maximum=1)
     )
     Vector2D=&Vector2DFromFloat(Value=&FloatFromCurve(
       FloatCurve=DataInterface() {propertyValues {Curve {keys [<at 0 0.1 RCIM_Cubic>, {arriveTangent=3.4967, tangentMode=RCTM_User, time=1, value=2.8}]}}}
       Scale Curve=&RandomRangeFloat(Minimum=0.6, Maximum=0.7)
      ))
    )) {module=ScaleSpriteSize001}
   &CameraOffset(Camera Offset Amount=&RandomRangeFloat(Minimum=1, Maximum=15))
  ]
  rendererClasses [&NiagaraSpriteRendererProperties]
  renderers [{rendererClass=&NiagaraSpriteRendererProperties}]
 }
}
summary {
 systemName=NS_Emote_CrumbViolin_SteamBurst
 emitters [
  {emitterName=SteamBursts, rendererClasses [&NiagaraSpriteRendererProperties]}
  {emitterName=Single_Burst_Late, rendererClasses [&NiagaraSpriteRendererProperties]}
  {emitterName=Single_Burst_Early, rendererClasses [&NiagaraSpriteRendererProperties]}
 ]
 userVariables [
  User.bIsFrontEnd {type=NiagaraBool}
  User.bIsFrontEndPreview {type=NiagaraBool}
  User.bIsLargeBodyPawn {type=NiagaraBool}
  User.Reveal {type=NiagaraFloat, defaultValue=0.5, description=""}
 ]
}
